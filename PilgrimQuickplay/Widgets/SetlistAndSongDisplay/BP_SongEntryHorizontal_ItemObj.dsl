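; Item object for a horizontal song entry in the Pilgrim Quickplay setlist and song display, holding a catalog song's data and setlist state.
; InitializeData stores the catalog item and its song metadata, then asks the custom setlist by short name whether anyone has added the song.
; The result sets the added-by-player state and selects one of two display statuses, clearing the player state when nobody has added it.
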(fn InitializeData
  (params (PilgrimSongCatalogItem :type Pilgrim%20Song%20Catalog%20Item%20VM%20Object%20Reference) (CustomSetlist :type Pilgrim%20Custom%20Setlist%20Object%20Reference))
  (bind _customsetlist (Variables|GetCustomSetlist))
  (bind _output_get (Variables|Default|SetPilgrimSongCatalogItem PilgrimSongCatalogItem))
  (Variables|Default|SetSongMetadata (Item|GetSongMetadata _output_get))
  (if (.ReturnValue (Class|PilgrimCustomSetlist|HasAnyoneAddedSong _customsetlist (Meta|GetShortName (Item|GetSongMetadata _output_get))))
    (Variables|Default|SetSongAddedByPlayerState (Class|PilgrimCustomSetlist|HasAnyoneAddedSong _customsetlist (Meta|GetShortName (Item|GetSongMetadata _output_get))))
    (Variables|Default|SetDisplayStatus "NewEnumerator0")
    (else
      (Variables|Default|SetSongAddedByPlayerState 0)
      (Variables|Default|SetDisplayStatus "NewEnumerator2"))))
