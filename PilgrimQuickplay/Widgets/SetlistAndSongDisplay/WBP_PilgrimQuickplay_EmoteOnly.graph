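# Small Pilgrim Quickplay setlist and song display widget that shows an "Emote Only" notice, styled as a missing-part alert.
# A horizontal box holds a tinted missing-parts alert icon, a narrow spacer and a semi-transparent Common text block that fills the remaining width.
# The text clips on demand and uses the setlist text scroll style.
# The widget is purely visual, with no variables and only unconnected default event nodes.
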WidgetBlueprint /PilgrimQuickplay/Widgets/SetlistAndSongDisplay/WBP_PilgrimQuickplay_EmoteOnly
graphs [EventGraph {nodes [
   Event PreConstruct()
   Event Construct()
   Event Tick()
  ]}]
variables []
widgets {
 info {inheritedWidgetCount=0, parentClass=&UserWidget, rootWidgetClass=&HorizontalBox, widgetCount=4}
 widgets [
  HB_MissingPart {widgetClassPath=&HorizontalBox, widget=HB_MissingPart}
  Image_MissingPartIcon {
   widgetClassPath=&Image
   parent=HB_MissingPart
   widget=Image_MissingPartIcon
   properties {
    visibility=SelfHitTestInvisible
    brush {
     bIsDynamicallyLoaded=false
     drawAs=Image
     imageType=FullColor
     mirroring=NoMirror
     resourceName=None
     resourceObject=/PilgrimQuickplay/UI/MI_UI_MissingPartsAlertIcon
     tiling=NoTile
     imageSize=<23 23>
     margin=<ltrb 0 0 0 0>
     outlineSettings {
      bUseBrushTransparency=false
      roundingType=HalfHeightRadius
      width=0
      color {colorUseRule=UseColor_Specified, specifiedColor=<rgba 0 0 0 0>}
      cornerRadii=<0 0 0 0>
     }
     tintColor {colorUseRule=UseColor_Specified, specifiedColor=<rgba 0.9301 1 0.0103 1>}
     uVRegion {bIsValid=false, max=<0 0>, min=<0 0>}
    }
   }
   slot {class=HorizontalBoxSlot, properties {horizontalAlignment=HAlign_Left, verticalAlignment=VAlign_Center}}
  }
  Spacer_65 {
   widgetClassPath=&Spacer
   parent=HB_MissingPart
   widget=Spacer_65
   properties {size=<4 1>}
   slot {class=HorizontalBoxSlot, properties {horizontalAlignment=HAlign_Left, verticalAlignment=VAlign_Center}}
  }
  Text_EmoteOnly {
   widgetClassPath=&CommonTextBlock
   parent=HB_MissingPart
   widget=Text_EmoteOnly
   properties {
    clipping=OnDemand
    scrollStyle=/PilgrimQuickplay/Widgets/ScrollStyles/TextScrollStyle_PilgrimQuickplay_Setlist.TextScrollStyle_PilgrimQuickplay_Setlist_C
    text=Emote Only
    colorAndOpacity {colorUseRule=UseColor_Specified, specifiedColor=<rgba 1 1 1 0.5>}
    font {
     bForceMonospaced=false
     bMaterialIsStencil=false
     fontMaterial=null
     fontObject=&Roboto
     letterSpacing=0
     monospacedWidth=1
     size=13
     skewAmount=0
     typefaceFontName=Black
     outlineSettings {
      bApplyOutlineToDropShadows=false
      bMiteredCorners=false
      bSeparateFillAlpha=false
      outlineMaterial=null
      outlineSize=0
      outlineColor=<rgba 0 0 0 1>
     }
    }
    shadowOffset=<0 0>
   }
   slot {class=HorizontalBoxSlot, properties {horizontalAlignment=HAlign_Left, verticalAlignment=VAlign_Center, size {sizeRule=Fill, value=1}}}
  }
 ]
}
