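# Quickplay song entry widget that shows one instrument's intensity as an icon and a row of seven sheared pips.
# Update Intensity for Instrument sets the icon through a material scalar and recolors pips, dimming all for a missing track.
# PreConstruct previews intensity at design time and shows the optional label and background only when the BattleStage.ShouldCheckScorePotential CVar check passes.
# PreConstruct and Construct both apply small or large sizing.
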WidgetBlueprint /PilgrimQuickplay/Widgets/SetlistAndSongDisplay/WBP_PilgrimQuickplay_SongIntensityEntry
graphs [
 EventGraph {nodes [
   UpdateMeterLabelAndBackground() {is=title K2Node_CustomEvent}
   > n1 = Branch(Condition="float > float"(A=GetNumericCVar!(CVar:String=BattleStage.ShouldCheckScorePotential) -> (OutValue:float, ReturnValue:bool)))
   > SetVisibility(
    self=Get Common_Text_Label() -> (Common_Text_Label:Common Text Object Reference)
    InVisibility=Select(Option 0=Collapsed, Option 1=SelfHitTestInvisible, Index=Get ShowMeterLabel() -> (ShowMeterLabel:bool))
   )
   > SetVisibility(
    self=Get OptionalBackground() -> (OptionalBackground:Common Border Object Reference)
    InVisibility=Select(Option 0=Collapsed, Option 1=SelfHitTestInvisible, Index=Get ShowBackground() -> (ShowBackground:bool))
   )
   > n2 = GetDynamicMaterial(self:Border Object Reference=Get OptionalBackground())
   > SetScalarParameterValue(self=@n2, ParameterName=Corner_BottomLeft, Value=0.3)
   > SetScalarParameterValue(self=@n2, ParameterName=Corner_TopRight, Value=0.3)
   > SetScalarParameterValue(self=@n2, ParameterName=Corner_BottomRight, Value=0.2)
   > SetScalarParameterValue(self=@n2, ParameterName=Corner_TopLeft, Value=0.2)
   SetVisibility(execute=@n1.else, self=Get Common_Text_Label(), InVisibility=Collapsed)
   > SetVisibility(self=Get OptionalBackground(), InVisibility=Collapsed)
   n3 = Event PreConstruct()
   > Parent: PreConstruct(IsDesignTime=@n3.IsDesignTime)
   > n4 = Branch(Condition=@n3.IsDesignTime)
   > n5 = SetIntensity!(execute, self, Intensity:int=Get PreviewIntensity() -> (PreviewIntensity:int=0)) -> (then)
   n6 = UpdateMeterLabelAndBackground(execute=[@n5.then, @n4.else]) {is=title K2Node_CallFunction}
   SetSize!(execute=[@n6.then, @n7.then], self, isLarge:bool=Get isLarge() -> (isLarge:bool)) -> (then)
   n7 = Event Construct()
  ]}
 GetIntensityPipByIndex {nodes [
   n1 = GetIntensityPipByIndex!() -> (then, PipIndex:int) {is=title K2Node_FunctionEntry}
   > n2 = Branch(Condition="In Range (Integer)"!(Value:int=@n1, Min:int=0, Max:int=6, InclusiveMin:bool=true, InclusiveMax:bool=true) -> (ReturnValue:bool))
   > Return Node!(
    execute
    IntensityMark:Border Object Reference=Select!(
     Option 0:Border Object Reference=Get Intensited1() -> (Intensited1:Border Object Reference)
     Option 1:Border Object Reference=Get Intensited2() -> (Intensited2:Border Object Reference)
     Option 2:Border Object Reference=Get Intensited3() -> (Intensited3:Border Object Reference)
     Option 3:Border Object Reference=Get Intensited4() -> (Intensited4:Border Object Reference)
     Option 4:Border Object Reference=Get Intensited5() -> (Intensited5:Border Object Reference)
     Option 5:Border Object Reference=Get Intensited6() -> (Intensited6:Border Object Reference)
     Option 6:Border Object Reference=Get Intensited7() -> (Intensited7:Border Object Reference)
     Index:int=@n1
    ) -> (ReturnValue:Border Object Reference)
   )
   PrintString!(
    execute=@n2.else
    InString=Append!(A="Tried to select a song intensity pip with an index of (", B="To String (Integer)"(InInt=@n1), C:String="), out of range 0-6") -> (ReturnValue)
    bPrintToScreen
    bPrintToLog
    TextColor
    Duration
    Key=TTSASIPWAIO2983497
   ) -> (then)
  ]}
 HideShowIcon {nodes [
   n1 = HideShowIcon!() -> (then, Show:bool=~) {is=title K2Node_FunctionEntry}
   > SetVisibility(
    self=Get Instrument() -> (Instrument:Image Object Reference)
    InVisibility=Select!(Option 0:ESlateVisibility Enum=Collapsed, Option 1:ESlateVisibility Enum=SelfHitTestInvisible, Index:bool=@n1) -> (ReturnValue:ESlateVisibility Enum)
   )
  ]}
 SetInstrumentIcon {nodes [
   n1 = SetInstrumentIcon!() -> (then, InstrumentType:EPilgrimTrackType Enum=TrackGuitar, StartingInstrument:ESparksInstrumentType Enum=Guitar) {is=title K2Node_FunctionEntry}
   > UpdateInstrumentIcon!(
    execute
    self
    Instrument:int=GetInstrumentIconIndex!(
     InstrumentType:EPilgrimTrackType Enum=@n1
     SongMetaData:Pilgrim Song Metadata Object Reference
     StartingInstrumentIndex:ESparksInstrumentType Enum=@n1.StartingInstrument
     bUseKeyTarIcon:bool=true
    ) -> (InstrumentIdexIcon:int=0)
   ) -> (then)
  ]}
 SetIntensity {nodes [
   n1 = SetIntensity!() -> (then, Intensity:int) {is=title K2Node_FunctionEntry}
   > n2 = Branch(Condition="Equal (Integer)"(A=@n1, B=GetMissingTrackIntensity!() -> (ReturnValue:int=0)))
   > n3 = For Loop(LastIndex=Get MaxIntensity() -> (MaxIntensity:int=0))
   > n4 = GetIntensityPipByIndex!(execute, self, PipIndex:int=@n3) -> (then, IntensityMark:Border Object Reference) {is=title K2Node_CallFunction}
   > SetBrushColor(self=@n4, InBrushColor=(R=1,G=1,B=1,A=0.08))
   n5 = For Loop(execute=@n2.else, LastIndex=Get Intensity() -> (Intensity:int=0))
   n6 = For Loop(execute=@n5.Completed, FirstIndex=int + int(A=Get Intensity() -> (Intensity=0)), LastIndex=Get MaxIntensity() -> (MaxIntensity=0))
   > n7 = GetIntensityPipByIndex(PipIndex=@n6)
   > n8 = Branch(Condition=Get Is3D() -> (Is3D:bool))
   > SetBrushColor(self=@n7, InBrushColor=(R=1,G=1,B=1,A=0.08))
   SetBrushColor(execute=@n8.else, self=@n7, InBrushColor=(R=1,G=1,B=1,A=0.2))
   n9 = GetIntensityPipByIndex(execute=@n5.LoopBody, PipIndex=@n5)
   > SetBrushColor(self=@n9, InBrushColor=(R=1,G=1,B=1,A=1))
  ]}
 SetSize {nodes [
   n1 = SetSize!() -> (then, isLarge:bool=~) {is=title K2Node_FunctionEntry}
   n2 = Select(Option 0:double=32, Option 1:double=48, Index=@n1) -> (ReturnValue:double)
   > SetDesiredSizeOverride!(execute, self:Image Object Reference=Get Instrument(), DesiredSize:Vector 2D Structure=MakeVector2D(X=@n2, Y=@n2)) -> (then)
   > n3 = For Each Loop(Array:Array of Border Object References=Make Array!(
     "[0]":Border Object Reference=Get Intensited1()
     "[1]":Border Object Reference=Get Intensited2()
     "[2]":Border Object Reference=Get Intensited3()
     "[3]":Border Object Reference=Get Intensited4()
     "[4]":Border Object Reference=Get Intensited5()
     "[5]":Border Object Reference=Get Intensited6()
     "[6]":Border Object Reference=Get Intensited7()
    ) -> (Array:Array of Border Object References)) -> (Array Element:Border Object Reference)
   > Set Padding!(
    execute
    Padding:Margin Structure=Make Margin(Left=Select(Option 0=8, Option 1=12, Index=@n1), Top=Select(Option 0=20, Option 1=32, Index=@n1), Right=0, Bottom=0)
    self:Border Object Reference=@n3
   ) -> (then, Output_Get:Margin Structure)
  ]}
 Update Intensity for Instrument {nodes [
   n1 = Update Intensity for Instrument!() -> (
    then
    Intensity:int
    TrackType:EPilgrimTrackType Enum=TrackGuitar
    StartingInstrument:ESparksInstrumentType Enum=Guitar
    Is3D_deprecated:bool=~
    SongMetadata:Pilgrim Song Metadata Object Reference
   ) {is=title K2Node_FunctionEntry}
   > SetInstrumentIcon!(execute, self, InstrumentType:EPilgrimTrackType Enum=@n1.TrackType, StartingInstrument:ESparksInstrumentType Enum=@n1.StartingInstrument) -> (then) {is=title K2Node_CallFunction}
   > SetIntensity(Intensity=@n1) {is=title K2Node_CallFunction}
  ]}
 UpdateInstrumentIcon {nodes [
   n1 = UpdateInstrumentIcon!() -> (then, Instrument:int) {is=title K2Node_FunctionEntry}
   > n2 = GetDynamicMaterial(self:Image Object Reference=Get Instrument())
   > SetScalarParameterValue(self=@n2, ParameterName=Icon, Value="To Float (Integer)"(InInt=@n1))
  ]}
]
variables [
 Is3D {type=boolean}
 isLarge {type=boolean}
 MaxIntensity {type=integer, default=6}
 PreviewIntensity {type=integer, default=5}
 ShowBackground {type=boolean}
 ShowMeterLabel {type=boolean}
]
widgets {
 info {inheritedWidgetCount=0, parentClass=&UserWidget, rootWidgetClass=&Overlay, widgetCount=16}
 widgets [
  Overlay_34 {widgetClassPath=&Overlay, widget=Overlay_34}
  OptionalBackground {
   widgetClassPath=&CommonBorder
   bIsVariable=true
   parent=Overlay_34
   widget=OptionalBackground
   properties {
    visibility=Collapsed
    background {
     bIsDynamicallyLoaded=false
     drawAs=Image
     imageType=FullColor
     mirroring=NoMirror
     resourceName=None
     resourceObject=/PilgrimCore/Widgets/Materials/MI_Pilgrim_TabBacking_Panel
     tiling=NoTile
     imageSize=<32 32>
     margin=<ltrb 0 0 0 0>
     outlineSettings {
      bUseBrushTransparency=false
      roundingType=HalfHeightRadius
      width=0
      color {colorUseRule=UseColor_Specified, specifiedColor=<rgba 0 0 0 0>}
      cornerRadii=<0 0 0 0>
     }
     tintColor {colorUseRule=UseColor_Specified, specifiedColor=<rgba 1 1 1 1>}
     uVRegion {bIsValid=false, max=<0 0>, min=<0 0>}
    }
   }
   slot {class=OverlaySlot, properties {horizontalAlignment=HAlign_Fill, verticalAlignment=VAlign_Fill, padding=<ltrb -4 -2 -6 -4>}}
  }
  IntensityHorizontalBox {
   widgetClassPath=&HorizontalBox
   parent=Overlay_34
   widget=IntensityHorizontalBox
   slot {class=OverlaySlot, properties {horizontalAlignment=HAlign_Fill, verticalAlignment=VAlign_Fill}}
  }
  CanvasPanel_0 {
   widgetClassPath=&CanvasPanel
   parent=IntensityHorizontalBox
   widget=CanvasPanel_0
   slot {
    class=HorizontalBoxSlot
    properties {horizontalAlignment=HAlign_Center, verticalAlignment=VAlign_Center, padding=<ltrb 0 0 -2 0>, size {sizeRule=Automatic, value=0.3}}
   }
  }
  Instrument {
   widgetClassPath=&Image
   bIsVariable=true
   parent=CanvasPanel_0
   widget=Instrument
   properties {
    visibility=SelfHitTestInvisible
    brush {
     bIsDynamicallyLoaded=false
     drawAs=Image
     imageType=FullColor
     mirroring=NoMirror
     resourceName=None
     resourceObject=/PilgrimQuickplay/Widgets/Materials/M_UI_Icon_Instrument_Type_MSDF
     tiling=NoTile
     imageSize=<32 32>
     margin=<ltrb 0 0 0 0>
     outlineSettings {
      bUseBrushTransparency=false
      roundingType=HalfHeightRadius
      width=0
      color {colorUseRule=UseColor_Specified, specifiedColor=<rgba 0 0 0 0>}
      cornerRadii=<0 0 0 0>
     }
     tintColor {colorUseRule=UseColor_Specified, specifiedColor=<rgba 1 1 1 1>}
     uVRegion {bIsValid=false, max=<0 0>, min=<0 0>}
    }
   }
   slot {class=CanvasPanelSlot, properties {layoutData {alignment=<0 0>, anchors {maximum=<0 0>, minimum=<0 0>}, offsets=<ltrb 0 0 30 30>}}}
  }
  MissingPartIcon {
   widgetClassPath=&Image
   bIsVariable=true
   parent=CanvasPanel_0
   widget=MissingPartIcon
   properties {
    visibility=Hidden
    brush {
     bIsDynamicallyLoaded=false
     drawAs=Image
     imageType=FullColor
     mirroring=NoMirror
     resourceName=None
     resourceObject=/PilgrimQuickplay/UI/MI_UI_MissingPartsAlertIcon
     tiling=NoTile
     imageSize=<32 32>
     margin=<ltrb 0 0 0 0>
     outlineSettings {
      bUseBrushTransparency=false
      roundingType=HalfHeightRadius
      width=0
      color {colorUseRule=UseColor_Specified, specifiedColor=<rgba 0 0 0 0>}
      cornerRadii=<0 0 0 0>
     }
     tintColor {colorUseRule=UseColor_Specified, specifiedColor=<rgba 0.9301 1 0.0103 1>}
     uVRegion {bIsValid=false, max=<0 0>, min=<0 0>}
    }
   }
   slot {class=CanvasPanelSlot, properties {layoutData {alignment=<0 0>, anchors {maximum=<1 1>, minimum=<1 1>}, offsets=<ltrb -18 -18 18 18>}}}
  }
  VerticalBox_71 {widgetClassPath=&VerticalBox, parent=IntensityHorizontalBox, widget=VerticalBox_71, slot {class=HorizontalBoxSlot}}
  Common_Text_Label {
   widgetClassPath=&CommonTextBlock
   bIsVariable=true
   parent=VerticalBox_71
   widget=Common_Text_Label
   properties {
    justification=Center
    text=Intensity
    visibility=HitTestInvisible
    colorAndOpacity {colorUseRule=UseColor_Specified, specifiedColor=<rgba 0.5 0.8333 1 1>}
    font {
     bForceMonospaced=false
     bMaterialIsStencil=false
     fontMaterial=null
     fontObject=&HeadingNow
     letterSpacing=0
     monospacedWidth=1
     size=7.5
     skewAmount=0
     typefaceFontName=75 Medium Italic
     outlineSettings {
      bApplyOutlineToDropShadows=false
      bMiteredCorners=false
      bSeparateFillAlpha=false
      outlineMaterial=null
      outlineSize=0
      outlineColor=<rgba 0 0 0 1>
     }
    }
    shadowOffset=<0 0>
   }
   slot {
    class=VerticalBoxSlot
    properties {horizontalAlignment=HAlign_Left, verticalAlignment=VAlign_Center, padding=<ltrb 4 0 0 0>, size {sizeRule=Fill, value=1}}
   }
  }
  HorizontalBox_99 {
   widgetClassPath=&HorizontalBox
   parent=VerticalBox_71
   widget=HorizontalBox_99
   slot {class=VerticalBoxSlot, properties {size {sizeRule=Fill, value=1}}}
  }
  Intensited1 {
   widgetClassPath=&Border
   bIsVariable=true
   parent=HorizontalBox_99
   widget=Intensited1
   properties {
    background {
     bIsDynamicallyLoaded=false
     drawAs=Image
     imageType=NoImage
     mirroring=NoMirror
     resourceName=None
     resourceObject=/PilgrimCore/Widgets/Materials/M_UI_Pilgrim_IntensityPip_Fill
     tiling=NoTile
     imageSize=<0 0>
     margin=<ltrb 0.5 0.5 0.5 0.5>
     outlineSettings {
      bUseBrushTransparency=false
      roundingType=HalfHeightRadius
      width=0
      color {colorUseRule=UseColor_Specified, specifiedColor=<rgba 0 0 0 0>}
      cornerRadii=<0 0 0 0>
     }
     tintColor {colorUseRule=UseColor_Specified, specifiedColor=<rgba 1 1 1 1>}
     uVRegion {bIsValid=false, max=<0 0>, min=<0 0>}
    }
    padding=<ltrb 8 20 0 0>
    renderTransform {angle=0, scale=<1 1>, shear=<-7 0>, translation=<0 0>}
   }
   slot {class=HorizontalBoxSlot, properties {horizontalAlignment=HAlign_Center, verticalAlignment=VAlign_Center, padding=<ltrb 3 0 0 0>}}
  }
  Intensited2 {
   widgetClassPath=&Border
   bIsVariable=true
   parent=HorizontalBox_99
   widget=Intensited2
   properties {
    background {
     bIsDynamicallyLoaded=false
     drawAs=Image
     imageType=NoImage
     mirroring=NoMirror
     resourceName=None
     resourceObject=&M_UI_Pilgrim_IntensityPip_Fill
     tiling=NoTile
     imageSize=<0 0>
     margin=<ltrb 0.5 0.5 0.5 0.5>
     outlineSettings {
      bUseBrushTransparency=false
      roundingType=HalfHeightRadius
      width=0
      color {colorUseRule=UseColor_Specified, specifiedColor=<rgba 0 0 0 0>}
      cornerRadii=<0 0 0 0>
     }
     tintColor {colorUseRule=UseColor_Specified, specifiedColor=<rgba 1 1 1 1>}
     uVRegion {bIsValid=false, max=<0 0>, min=<0 0>}
    }
    padding=<ltrb 8 20 0 0>
    renderTransform {angle=0, scale=<1 1>, shear=<-7 0>, translation=<0 0>}
   }
   slot {class=HorizontalBoxSlot, properties {horizontalAlignment=HAlign_Center, verticalAlignment=VAlign_Center, padding=<ltrb 3 0 0 0>}}
  }
  Intensited3 {
   widgetClassPath=&Border
   bIsVariable=true
   parent=HorizontalBox_99
   widget=Intensited3
   properties {
    background {
     bIsDynamicallyLoaded=false
     drawAs=Image
     imageType=NoImage
     mirroring=NoMirror
     resourceName=None
     resourceObject=&M_UI_Pilgrim_IntensityPip_Fill
     tiling=NoTile
     imageSize=<0 0>
     margin=<ltrb 0.5 0.5 0.5 0.5>
     outlineSettings {
      bUseBrushTransparency=false
      roundingType=HalfHeightRadius
      width=0
      color {colorUseRule=UseColor_Specified, specifiedColor=<rgba 0 0 0 0>}
      cornerRadii=<0 0 0 0>
     }
     tintColor {colorUseRule=UseColor_Specified, specifiedColor=<rgba 1 1 1 1>}
     uVRegion {bIsValid=false, max=<0 0>, min=<0 0>}
    }
    padding=<ltrb 8 20 0 0>
    renderTransform {angle=0, scale=<1 1>, shear=<-7 0>, translation=<0 0>}
   }
   slot {class=HorizontalBoxSlot, properties {horizontalAlignment=HAlign_Center, verticalAlignment=VAlign_Center, padding=<ltrb 3 0 0 0>}}
  }
  Intensited4 {
   widgetClassPath=&Border
   bIsVariable=true
   parent=HorizontalBox_99
   widget=Intensited4
   properties {
    background {
     bIsDynamicallyLoaded=false
     drawAs=Image
     imageType=NoImage
     mirroring=NoMirror
     resourceName=None
     resourceObject=&M_UI_Pilgrim_IntensityPip_Fill
     tiling=NoTile
     imageSize=<32 32>
     margin=<ltrb 0 0 0 0>
     outlineSettings {
      bUseBrushTransparency=false
      roundingType=HalfHeightRadius
      width=0
      color {colorUseRule=UseColor_Specified, specifiedColor=<rgba 0 0 0 0>}
      cornerRadii=<0 0 0 0>
     }
     tintColor {colorUseRule=UseColor_Specified, specifiedColor=<rgba 1 1 1 1>}
     uVRegion {bIsValid=false, max=<0 0>, min=<0 0>}
    }
    padding=<ltrb 8 20 0 0>
    renderTransform {angle=0, scale=<1 1>, shear=<-7 0>, translation=<0 0>}
   }
   slot {class=HorizontalBoxSlot, properties {horizontalAlignment=HAlign_Center, verticalAlignment=VAlign_Center, padding=<ltrb 3 0 0 0>}}
  }
  Intensited5 {
   widgetClassPath=&Border
   bIsVariable=true
   parent=HorizontalBox_99
   widget=Intensited5
   properties {
    background {
     bIsDynamicallyLoaded=false
     drawAs=Image
     imageType=NoImage
     mirroring=NoMirror
     resourceName=None
     resourceObject=&M_UI_Pilgrim_IntensityPip_Fill
     tiling=NoTile
     imageSize=<0 0>
     margin=<ltrb 0.5 0.5 0.5 0.5>
     outlineSettings {
      bUseBrushTransparency=false
      roundingType=HalfHeightRadius
      width=0
      color {colorUseRule=UseColor_Specified, specifiedColor=<rgba 0 0 0 0>}
      cornerRadii=<0 0 0 0>
     }
     tintColor {colorUseRule=UseColor_Specified, specifiedColor=<rgba 1 1 1 1>}
     uVRegion {bIsValid=false, max=<0 0>, min=<0 0>}
    }
    padding=<ltrb 8 20 0 0>
    renderTransform {angle=0, scale=<1 1>, shear=<-7 0>, translation=<0 0>}
   }
   slot {class=HorizontalBoxSlot, properties {horizontalAlignment=HAlign_Center, verticalAlignment=VAlign_Center, padding=<ltrb 3 0 0 0>}}
  }
  Intensited6 {
   widgetClassPath=&Border
   bIsVariable=true
   parent=HorizontalBox_99
   widget=Intensited6
   properties {
    background {
     bIsDynamicallyLoaded=false
     drawAs=Image
     imageType=NoImage
     mirroring=NoMirror
     resourceName=None
     resourceObject=&M_UI_Pilgrim_IntensityPip_Fill
     tiling=NoTile
     imageSize=<0 0>
     margin=<ltrb 0.5 0.5 0.5 0.5>
     outlineSettings {
      bUseBrushTransparency=false
      roundingType=HalfHeightRadius
      width=0
      color {colorUseRule=UseColor_Specified, specifiedColor=<rgba 0 0 0 0>}
      cornerRadii=<0 0 0 0>
     }
     tintColor {colorUseRule=UseColor_Specified, specifiedColor=<rgba 1 1 1 1>}
     uVRegion {bIsValid=false, max=<0 0>, min=<0 0>}
    }
    padding=<ltrb 8 20 0 0>
    renderTransform {angle=0, scale=<1 1>, shear=<-7 0>, translation=<0 0>}
   }
   slot {class=HorizontalBoxSlot, properties {horizontalAlignment=HAlign_Center, verticalAlignment=VAlign_Center, padding=<ltrb 3 0 0 0>}}
  }
  Intensited7 {
   widgetClassPath=&Border
   bIsVariable=true
   parent=HorizontalBox_99
   widget=Intensited7
   properties {
    background {
     bIsDynamicallyLoaded=false
     drawAs=Image
     imageType=NoImage
     mirroring=NoMirror
     resourceName=None
     resourceObject=&M_UI_Pilgrim_IntensityPip_Fill
     tiling=NoTile
     imageSize=<0 0>
     margin=<ltrb 0.5 0.5 0.5 0.5>
     outlineSettings {
      bUseBrushTransparency=false
      roundingType=HalfHeightRadius
      width=0
      color {colorUseRule=UseColor_Specified, specifiedColor=<rgba 0 0 0 0>}
      cornerRadii=<0 0 0 0>
     }
     tintColor {colorUseRule=UseColor_Specified, specifiedColor=<rgba 1 1 1 1>}
     uVRegion {bIsValid=false, max=<0 0>, min=<0 0>}
    }
    brushColor=<rgba 1 1 1 0.25>
    padding=<ltrb 8 20 0 0>
    renderTransform {angle=0, scale=<1 1>, shear=<-7 0>, translation=<0 0>}
   }
   slot {class=HorizontalBoxSlot, properties {horizontalAlignment=HAlign_Center, verticalAlignment=VAlign_Center, padding=<ltrb 3 0 0 0>}}
  }
 ]
}
