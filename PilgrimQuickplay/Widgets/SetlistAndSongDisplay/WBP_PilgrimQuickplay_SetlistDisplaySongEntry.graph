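# Setlist song entry widget for Pilgrim Quickplay, showing slot number, album art, title, artist and track intensity.
# PreConstruct applies the slot text and display toggles; contents start hidden.
# UpdateEntry reveals the row, fills text from song metadata, updates intensity for the given track type and lazy-loads album art.
# It also feeds metadata and selected part to a MissingParts view model; a switcher holds the intensity and emote-only widgets.
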WidgetBlueprint /PilgrimQuickplay/Widgets/SetlistAndSongDisplay/WBP_PilgrimQuickplay_SetlistDisplaySongEntry
graphs [
 EventGraph {nodes [
   n1 = Event PreConstruct()
   > Parent: PreConstruct(IsDesignTime=@n1.IsDesignTime)
   > SetVisibility(
    self=Get IntensityHorzontalBox() -> (IntensityHorzontalBox:Horizontal Box Object Reference)
    InVisibility=Select(Option 0=HitTestInvisible, Option 1=Collapsed, Index=Get HideIntensity() -> (HideIntensity:bool))
   )
   n2 = Get Intensity() -> (Intensity:WBP Pilgrim Quickplay Song Intensity Entry Object Reference)
   > Set ShowMeterLabel!(
    execute
    ShowMeterLabel:bool=Get ShowIntensityLabel() -> (ShowIntensityLabel:bool)
    self:WBP Pilgrim Quickplay Song Intensity Entry Object Reference=@n2
   ) -> (then, Output_Get:bool)
   > UpdateMeterLabelAndBackground(self:WBP Pilgrim Quickplay Song Intensity Entry Object Reference=@n2)
   > SetVisibility(
    self=Get SlotNumberOverlay() -> (SlotNumberOverlay:Overlay Object Reference)
    InVisibility=Select(Option 0=Collapsed, Option 1=SelfHitTestInvisible, Index=Get ShouldShowSlotNumber() -> (ShouldShowSlotNumber:bool))
   )
   > Set Text(Text=Get SlotIDText() -> (SlotIDText:Text), self=Get SlotNumber() -> (SlotNumber:Common Text Object Reference))
   Event Tick()
  ]}
 HideSongEntry {nodes [
   HideSongEntry() {is=title K2Node_FunctionEntry}
   > SetVisibility(
    self=[
     Get Intensity()
     Get AlbumArt() -> (AlbumArt:Fort Lazy Image Object Reference)
     Get SlotNumberOverlay()
     Get SongArtist() -> (SongArtist:Common Text Object Reference)
     Get SongTitle() -> (SongTitle:Common Text Object Reference)
    ]
    InVisibility=Hidden
   )
  ]}
 ShowSongEntry {nodes [
   ShowSongEntry() {is=title K2Node_FunctionEntry}
   > SetVisibility(InVisibility=SelfHitTestInvisible)
   > SetVisibility(
    self=[
     Get Intensity()
     Get AlbumArt()
     Get ContentsHorizontalBox() -> (ContentsHorizontalBox:Horizontal Box Object Reference)
     Get SongArtist()
     Get SongTitle()
    ]
    InVisibility=HitTestInvisible
   )
   > Branch(Condition=Get ShouldShowSlotNumber())
   > SetVisibility(self=[Get SlotNumberOverlay(), Get SlotNumber()], InVisibility=SelfHitTestInvisible)
  ]}
 UpdateEntry {nodes [
   n1 = UpdateEntry!() -> (then, SongMetadata:Pilgrim Song Metadata Object Reference, InputPin1:EPilgrimTrackType Enum=TrackGuitar) {is=title K2Node_FunctionEntry}
   > ForceLayoutPrepass(self:Widget Object Reference)
   > Is Valid(InputObject=@n1)
   > ShowSongEntry() {is=title K2Node_CallFunction}
   > "SetText (Text)"(self=Get SongTitle(), InText=GetTitleText!(self:Sparks Song Data Object Reference=@n1) -> (ReturnValue:Text))
   > "SetText (Text)"(self=Get SongArtist(), InText=GetArtistText!(self:Sparks Song Data Object Reference=@n1) -> (ReturnValue:Text))
   > UpdateIntensity!(
    execute
    self
    Intensity:int=GetTrackIntensity!(self:Pilgrim Song Metadata Object Reference=@n1, TrackType:EPilgrimTrackType Enum=@n1.InputPin1) -> (ReturnValue:int=0)
    TrackType:EPilgrimTrackType Enum=@n1.InputPin1
    StartingInstrument:ESparksInstrumentType Enum=GetStartingInstrumentType!(self:Pilgrim Song Metadata Object Reference=@n1, TrackType:EPilgrimTrackType Enum=@n1.InputPin1) -> (ReturnValue:ESparksInstrumentType Enum=Guitar)
   ) -> (then)
   > LazyLoadAlbumArt!(execute, Song Metadata:Pilgrim Song Metadata Object Reference=@n1, Lazy Image:Fort Lazy Image Object Reference=Get AlbumArt()) -> (then)
   > ReceiveSongMetadata!(
    execute
    self:BP Pilgrim Quickplay Missing Parts VM Object Reference=Get BP_PilgrimQuickplay_MissingPartsVM() -> (BP_PilgrimQuickplay_MissingPartsVM:BP Pilgrim Quickplay Missing Parts VM Object Reference)
    Song Metadata:Pilgrim Song Metadata Object Reference=Get SongMetadata() -> (SongMetadata:Pilgrim Song Metadata Object Reference)
   ) -> (then)
  ]}
 UpdateIntensity {nodes [
   n1 = UpdateIntensity!() -> (then, Intensity:int, TrackType:EPilgrimTrackType Enum=TrackGuitar, StartingInstrument:ESparksInstrumentType Enum=Guitar) {is=title K2Node_FunctionEntry}
   > Update Intensity for Instrument!(
    execute
    self:WBP Pilgrim Quickplay Song Intensity Entry Object Reference=Get Intensity()
    Intensity:int=@n1
    TrackType:EPilgrimTrackType Enum=@n1.TrackType
    StartingInstrument:ESparksInstrumentType Enum=@n1.StartingInstrument
    Is3D_deprecated:bool=false
    SongMetadata:Pilgrim Song Metadata Object Reference
   ) -> (then)
   > n2 = Get!(execute:Exec) -> (then:Exec, else:Exec, BP_PilgrimQuickplay_MissingPartsVM:BP Pilgrim Quickplay Missing Parts VM Object Reference)
   > Set with Broadcast Current Part Selected!(
    execute
    Current Part Selected:EPilgrimTrackType Enum=Get TrackType() -> (TrackType:EPilgrimTrackType Enum=TrackGuitar)
    self:BP Pilgrim Quickplay Missing Parts VM Object Reference=@n2
   ) -> (then, Output_Get:EPilgrimTrackType Enum=TrackGuitar)
  ]}
]
variables [
 HideIntensity {type=boolean}
 ShouldShowSlotNumber {type=boolean, default=true}
 ShowIntensityLabel {type=boolean}
 SlotIDText {type=string, default="##"}
]
widgets {
 info {inheritedWidgetCount=0, parentClass=&UserWidget, rootWidgetClass=&Overlay, widgetCount=20}
 widgets [
  Overlay {widgetClassPath=&Overlay, widget=Overlay}
  EntryBorder {
   widgetClassPath=&CommonBorder
   bIsVariable=true
   parent=Overlay
   widget=EntryBorder
   properties {
    bOverrideAccessibleDefaults=true
    background {
     bIsDynamicallyLoaded=false
     drawAs=NoDrawType
     imageType=NoImage
     mirroring=NoMirror
     resourceName=None
     resourceObject=/Game/UI/GreyBox/white
     tiling=NoTile
     imageSize=<8 8>
     margin=<ltrb 0.1 0.1 0.1 0.1>
     outlineSettings {
      bUseBrushTransparency=false
      roundingType=HalfHeightRadius
      width=0
      color {colorUseRule=UseColor_Specified, specifiedColor=<rgba 0 0 0 0>}
      cornerRadii=<0 0 0 0>
     }
     tintColor {colorUseRule=UseColor_Specified, specifiedColor=<rgba 1 1 1 1>}
     uVRegion {bIsValid=false, max=<0 0>, min=<0 0>}
    }
    padding=<ltrb 0 0 0 0>
   }
   slot {class=OverlaySlot, properties {horizontalAlignment=HAlign_Fill, verticalAlignment=VAlign_Fill}}
  }
  Overlay_0 {widgetClassPath=&Overlay, parent=EntryBorder, widget=Overlay_0, slot {class=BorderSlot, properties {padding=<ltrb 0 0 0 0>}}}
  BackgroundImage {
   widgetClassPath=&Image
   bIsVariable=true
   parent=Overlay_0
   widget=BackgroundImage
   properties {
    brush {
     bIsDynamicallyLoaded=false
     drawAs=Image
     imageType=FullColor
     mirroring=NoMirror
     resourceName=None
     resourceObject=/PilgrimCore/Widgets/Materials/MI_UI_Underlayer_BackstageScreen
     tiling=NoTile
     imageSize=<32 32>
     margin=<ltrb 0 0 0 0>
     outlineSettings {
      bUseBrushTransparency=false
      roundingType=HalfHeightRadius
      width=0
      color {colorUseRule=UseColor_Specified, specifiedColor=<rgba 0 0 0 0>}
      cornerRadii=<0 0 0 0>
     }
     tintColor {colorUseRule=UseColor_Specified, specifiedColor=<rgba 1 1 1 1>}
     uVRegion {bIsValid=false, max=<0 0>, min=<0 0>}
    }
    colorAndOpacity=<rgba 1 1 1 0.3>
   }
   slot {class=OverlaySlot, properties {horizontalAlignment=HAlign_Fill, verticalAlignment=VAlign_Fill}}
  }
  ContentsHorizontalBox {
   widgetClassPath=&HorizontalBox
   bIsVariable=true
   parent=Overlay_0
   widget=ContentsHorizontalBox
   properties {visibility=Hidden}
   slot {class=OverlaySlot, properties {horizontalAlignment=HAlign_Fill, verticalAlignment=VAlign_Fill, padding=<ltrb 8 8 8 8>}}
  }
  SlotNumberOverlay {
   widgetClassPath=&Overlay
   bIsVariable=true
   parent=ContentsHorizontalBox
   widget=SlotNumberOverlay
   slot {class=HorizontalBoxSlot, properties {horizontalAlignment=HAlign_Left, padding=<ltrb 0 0 16 0>}}
  }
  SlotNumber {
   widgetClassPath=&CommonTextBlock
   bIsVariable=true
   parent=SlotNumberOverlay
   widget=SlotNumber
   properties {
    bIsScrollingEnabled=false
    clipping=ClipToBounds
    text=1
    visibility=HitTestInvisible
    font {
     bForceMonospaced=false
     bMaterialIsStencil=false
     fontMaterial=null
     fontObject=&HeadingNow
     letterSpacing=0
     monospacedWidth=1
     size=37.5
     skewAmount=0
     typefaceFontName=75 Medium
     outlineSettings {
      bApplyOutlineToDropShadows=false
      bMiteredCorners=false
      bSeparateFillAlpha=false
      outlineMaterial=null
      outlineSize=0
      outlineColor=<rgba 0 0 0 1>
     }
    }
    shadowOffset=<0 0>
   }
   slot {class=OverlaySlot, properties {horizontalAlignment=HAlign_Center, verticalAlignment=VAlign_Center}}
  }
  Image {
   widgetClassPath=&Image
   bIsVariable=true
   parent=SlotNumberOverlay
   widget=Image
   properties {
    visibility=SelfHitTestInvisible
    brush {
     bIsDynamicallyLoaded=false
     drawAs=Image
     imageType=FullColor
     mirroring=NoMirror
     resourceName=None
     resourceObject=&MI_UI_Underlayer_BackstageScreen
     tiling=NoTile
     imageSize=<64 32>
     margin=<ltrb 0 0 0 0>
     outlineSettings {
      bUseBrushTransparency=false
      roundingType=HalfHeightRadius
      width=0
      color {colorUseRule=UseColor_Specified, specifiedColor=<rgba 0 0 0 0>}
      cornerRadii=<0 0 0 0>
     }
     tintColor {colorUseRule=UseColor_Specified, specifiedColor=<rgba 1 1 1 1>}
     uVRegion {bIsValid=false, max=<0 0>, min=<0 0>}
    }
    colorAndOpacity=<rgba 1 1 1 0.3>
   }
   slot {class=OverlaySlot, properties {horizontalAlignment=HAlign_Fill, verticalAlignment=VAlign_Fill}}
  }
  ScaleBox_327 {
   widgetClassPath=&ScaleBox
   parent=ContentsHorizontalBox
   widget=ScaleBox_327
   properties {stretch=ScaleToFitY}
   slot {class=HorizontalBoxSlot, properties {horizontalAlignment=HAlign_Center}}
  }
  AlbumArt {
   widgetClassPath=&FortLazyImage
   bIsVariable=true
   parent=ScaleBox_327
   widget=AlbumArt
   properties {
    materialTextureParamName=T_AlbumArt
    brush {
     bIsDynamicallyLoaded=false
     drawAs=Image
     imageType=FullColor
     mirroring=NoMirror
     resourceName=None
     resourceObject=/SparksUICommon/Materials/MI_UI_Sparks_SetList_AlbumArt_Rounded
     tiling=NoTile
     imageSize=<256 256>
     margin=<ltrb 0 0 0 0>
     outlineSettings {
      bUseBrushTransparency=false
      roundingType=HalfHeightRadius
      width=0
      color {colorUseRule=UseColor_Specified, specifiedColor=<rgba 0 0 0 0>}
      cornerRadii=<0 0 0 0>
     }
     tintColor {colorUseRule=UseColor_Specified, specifiedColor=<rgba 1 1 1 1>}
     uVRegion {bIsValid=false, max=<0 0>, min=<0 0>}
    }
   }
   slot {class=ScaleBoxSlot}
  }
  SongInfoVerticalBox {
   widgetClassPath=&VerticalBox
   parent=ContentsHorizontalBox
   widget=SongInfoVerticalBox
   slot {class=HorizontalBoxSlot, properties {verticalAlignment=VAlign_Center, padding=<ltrb 16 0 10 0>, size {sizeRule=Fill, value=1}}}
  }
  ScaleBox_6 {
   widgetClassPath=&ScaleBox
   parent=SongInfoVerticalBox
   widget=ScaleBox_6
   properties {stretch=None, stretchDirection=DownOnly}
   slot {class=VerticalBoxSlot, properties {size {sizeRule=Fill, value=1}}}
  }
  Overlay_2 {
   widgetClassPath=&Overlay
   parent=ScaleBox_6
   widget=Overlay_2
   slot {class=ScaleBoxSlot, properties {horizontalAlignment=HAlign_Fill, verticalAlignment=VAlign_Bottom}}
  }
  SongTitle {
   widgetClassPath=&CommonTextBlock
   bIsVariable=true
   parent=Overlay_2
   widget=SongTitle
   properties {
    bIsScrollingEnabled=false
    clipping=ClipToBounds
    text=SONG TITLE IS QUITE LONG
    visibility=HitTestInvisible
    wrapTextAt=860
    font {
     bForceMonospaced=false
     bMaterialIsStencil=false
     fontMaterial=null
     fontObject=&HeadingNow
     letterSpacing=0
     monospacedWidth=1
     size=37.5
     skewAmount=0
     typefaceFontName=75 Medium
     outlineSettings {
      bApplyOutlineToDropShadows=false
      bMiteredCorners=false
      bSeparateFillAlpha=false
      outlineMaterial=null
      outlineSize=0
      outlineColor=<rgba 0 0 0 1>
     }
    }
    shadowOffset=<0 0>
   }
   slot {class=OverlaySlot, properties {verticalAlignment=VAlign_Bottom}}
  }
  SongArtist {
   widgetClassPath=&CommonTextBlock
   bIsVariable=true
   parent=SongInfoVerticalBox
   widget=SongArtist
   properties {
    clipping=OnDemand
    scrollStyle=/PilgrimQuickplay/Widgets/ScrollStyles/TextScrollStyle_PilgrimQuickplay_Setlist.TextScrollStyle_PilgrimQuickplay_Setlist_C
    text=ARTIST
    visibility=HitTestInvisible
    colorAndOpacity {colorUseRule=UseColor_Specified, specifiedColor=<rgba 0.3749 0.7544 1 0.771>}
    font {
     bForceMonospaced=false
     bMaterialIsStencil=false
     fontMaterial=null
     fontObject=&HeadingNow
     letterSpacing=0
     monospacedWidth=1
     size=22.5
     skewAmount=0
     typefaceFontName=74 Regular
     outlineSettings {
      bApplyOutlineToDropShadows=false
      bMiteredCorners=false
      bSeparateFillAlpha=false
      outlineMaterial=null
      outlineSize=0
      outlineColor=<rgba 0 0 0 1>
     }
    }
    shadowOffset=<0 0>
   }
   slot {class=VerticalBoxSlot, properties {horizontalAlignment=HAlign_Left, verticalAlignment=VAlign_Center, padding=<ltrb 5 0 5 0>}}
  }
  IntensityHorzontalBox {
   widgetClassPath=&HorizontalBox
   bIsVariable=true
   parent=SongInfoVerticalBox
   widget=IntensityHorzontalBox
   properties {bOverride_Cursor=true}
   slot {class=VerticalBoxSlot, properties {padding=<ltrb 0 8 0 0>, size {sizeRule=Automatic, value=0.5}}}
  }
  Switcher_Info {
   widgetClassPath=&CommonVisibilitySwitcher
   parent=IntensityHorzontalBox
   widget=Switcher_Info
   slot {class=HorizontalBoxSlot}
  }
  Intensity {
   widgetClassPath=&/WBP_PilgrimQuickplay_SongIntensityEntry
   bIsVariable=true
   parent=Switcher_Info
   widget=Intensity
   properties {is3D=true, isLarge=true, visibility=HitTestInvisible}
   slot {class=CommonVisibilitySwitcherSlot}
  }
  EmoteOnly {
   widgetClassPath=&/WBP_PilgrimQuickplay_EmoteOnly
   parent=Switcher_Info
   widget=EmoteOnly
   properties {accessibleBehavior=Auto, bCanChildrenBeAccessible=false, bOverride_Cursor=true}
   slot {class=CommonVisibilitySwitcherSlot, properties {horizontalAlignment=HAlign_Right, verticalAlignment=VAlign_Center}}
  }
  OptionalMeterEntry {
   widgetClassPath=&NamedSlot
   bIsVariable=true
   parent=IntensityHorzontalBox
   widget=OptionalMeterEntry
   properties {visibility=Collapsed}
   slot {class=HorizontalBoxSlot, properties {padding=<ltrb 32 0 0 0>}}
  }
 ]
}
