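# Pilgrim Quickplay setlist display widget showing the chosen songs in a two-column grid of six song entries under a song count title.
# UpdateCustomSetlist toggles a "Choose a Song" overlay for empty lists; otherwise it stores the list, updates entries by index and hides unused ones.
# The entry array is built once from the named widgets.
# UpdateAllIntensitiesOnInstrumentChanged refreshes each entry's intensity from the stored song metadata for the new instrument.
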WidgetBlueprint /PilgrimQuickplay/Widgets/SetlistAndSongDisplay/WBP_PilgrimQuickplay_SetlistDisplayCustom
graphs [
 EventGraph {nodes [
   Event Construct()
   > PopulateSongEntryArray()
   Event PreConstruct()
   Event Tick()
  ]}
 PopulateSongEntryArray {nodes [
   PopulateSongEntryArray() {is=title K2Node_FunctionEntry}
   > Branch(Condition=Is Empty!(TargetArray:Array of WBP Pilgrim Quickplay Setlist Display Song Entry Object References=Get SongEntries() -> (SongEntries:Array of WBP Pilgrim Quickplay Setlist Display Song Entry Object References)) -> (ReturnValue:bool))
   > Set SongEntries(SongEntries:Array of WBP Pilgrim Quickplay Setlist Display Song Entry Object References=Make Array!(
     "[0]":WBP Pilgrim Quickplay Setlist Display Song Entry Object Reference=Get Song1() -> (Song1:WBP Pilgrim Quickplay Setlist Display Song Entry Object Reference)
     "[1]":WBP Pilgrim Quickplay Setlist Display Song Entry Object Reference=Get Song2() -> (Song2:WBP Pilgrim Quickplay Setlist Display Song Entry Object Reference)
     "[2]":WBP Pilgrim Quickplay Setlist Display Song Entry Object Reference=Get Song3() -> (Song3:WBP Pilgrim Quickplay Setlist Display Song Entry Object Reference)
     "[3]":WBP Pilgrim Quickplay Setlist Display Song Entry Object Reference=Get Song4() -> (Song4:WBP Pilgrim Quickplay Setlist Display Song Entry Object Reference)
     "[4]":WBP Pilgrim Quickplay Setlist Display Song Entry Object Reference=Get Song5() -> (Song5:WBP Pilgrim Quickplay Setlist Display Song Entry Object Reference)
     "[5]":WBP Pilgrim Quickplay Setlist Display Song Entry Object Reference=Get Song6() -> (Song6:WBP Pilgrim Quickplay Setlist Display Song Entry Object Reference)
    ) -> (Array:Array of WBP Pilgrim Quickplay Setlist Display Song Entry Object References)) -> (Output_Get:Array of WBP Pilgrim Quickplay Setlist Display Song Entry Object References)
  ]}
 ToggleSetlistVisibility {nodes [
   n1 = ToggleSetlistVisibility!() -> (then, ShowSetlist:bool=~) {is=title K2Node_FunctionEntry}
   > SetVisibility(
    self=Get NoSongsChosenLabel() -> (NoSongsChosenLabel:Overlay Object Reference)
    InVisibility=Select(Option 0=HitTestInvisible, Option 1=Hidden, Index=@n1)
   )
  ]}
 UpdateAllIntensitiesOnInstrumentChanged {nodes [
   n1 = UpdateAllIntensitiesOnInstrumentChanged!() -> (then, NewInstrument:EPilgrimTrackType Enum=TrackGuitar) {is=title K2Node_FunctionEntry}
   > n2 = For Each Loop(Array:Array of WBP Pilgrim Quickplay Setlist Display Song Entry Object References=Get SongEntries()) -> (Array Element:WBP Pilgrim Quickplay Setlist Display Song Entry Object Reference)
   n3 = Get Song List() -> (Song List:Array of Pilgrim Song Metadata Object References)
   > Branch(Condition=Is Valid Index!(TargetArray:Array of Pilgrim Song Metadata Object References=@n3, IndexToTest:int=@n2.Array Index) -> (ReturnValue:bool))
   > UpdateSongEntryIntensityIndicators!(
    execute
    self
    SongEntry:WBP Pilgrim Quickplay Setlist Display Song Entry Object Reference=@n2
    SongMetadata:Pilgrim Song Metadata Object Reference="Get (a copy)"(Array:Array of Pilgrim Song Metadata Object References=@n3, Dimension 1=@n2.Array Index) -> (Output:Pilgrim Song Metadata Object Reference)
    InstrumentType:EPilgrimTrackType Enum=@n1
   ) -> (then)
  ]}
 UpdateCustomSetlist {nodes [
   n1 = UpdateCustomSetlist!() -> (then, SongList:Array of Pilgrim Song Metadata Object References, LocalInstrument:EPilgrimTrackType Enum=TrackGuitar) {is=title K2Node_FunctionEntry}
   > ForceLayoutPrepass(self:Widget Object Reference)
   n2 = Length(TargetArray:Array of Pilgrim Song Metadata Object References=@n1)
   > "SetText (Text)"(
    self=Get SetlistName() -> (SetlistName:Common Text Object Reference)
    InText=Format Text!(Format="LOCTEXT(\"Setlist Songs: {count}\")", count:int=@n2) -> (Result)
   )
   n3 = "integer > integer"(A=@n2)
   > ToggleSetlistVisibility!(execute, self, ShowSetlist:bool=@n3) -> (then) {is=title K2Node_CallFunction}
   > Branch(Condition=@n3)
   > Set Song List(Song List:Array of Pilgrim Song Metadata Object References=@n1) -> (Output_Get:Array of Pilgrim Song Metadata Object References)
   > PopulateSongEntryArray() {is=title K2Node_CallFunction}
   > n4 = For Each Loop(Array=Get SongEntries())
   > n5 = Branch(Condition=Is Valid Index(TargetArray=@n1, IndexToTest=@n4.Array Index))
   > UpdateSongEntry!(
    execute
    self
    SongEntry:WBP Pilgrim Quickplay Setlist Display Song Entry Object Reference=@n4
    SongMetadata:Pilgrim Song Metadata Object Reference="Get (a copy)"(Array=@n1, Dimension 1=@n4.Array Index)
    LocalPlayerInstrument:EPilgrimTrackType Enum=@n1.LocalInstrument
   ) -> (then)
   HideSongEntry(execute=@n5.else, self:WBP Pilgrim Quickplay Setlist Display Song Entry Object Reference=@n4)
  ]}
 UpdateSongEntry {nodes [
   n1 = UpdateSongEntry!() -> (
    then
    SongEntry:WBP Pilgrim Quickplay Setlist Display Song Entry Object Reference
    SongMetadata:Pilgrim Song Metadata Object Reference
    LocalPlayerInstrument:EPilgrimTrackType Enum=TrackGuitar
   ) {is=title K2Node_FunctionEntry}
   > n2 = Is Valid(InputObject=@n1)
   > UpdateEntry!(
    execute
    self:WBP Pilgrim Quickplay Setlist Display Song Entry Object Reference=@n1
    SongMetadata:Pilgrim Song Metadata Object Reference=@n1.SongMetadata
    InputPin1:EPilgrimTrackType Enum=@n1.LocalPlayerInstrument
   ) -> (then)
   HideSongEntry(execute=@n2.Is Not Valid, self=@n1)
  ]}
 UpdateSongEntryIntensityIndicators {nodes [
   n1 = UpdateSongEntryIntensityIndicators!() -> (
    then
    SongEntry:WBP Pilgrim Quickplay Setlist Display Song Entry Object Reference
    SongMetadata:Pilgrim Song Metadata Object Reference
    InstrumentType:EPilgrimTrackType Enum=TrackGuitar
   ) {is=title K2Node_FunctionEntry}
   > UpdateIntensity!(
    execute
    self:WBP Pilgrim Quickplay Setlist Display Song Entry Object Reference=@n1
    Intensity:int=GetTrackIntensity!(self:Pilgrim Song Metadata Object Reference=@n1.SongMetadata, TrackType:EPilgrimTrackType Enum=@n1.InstrumentType) -> (ReturnValue:int=0)
    TrackType:EPilgrimTrackType Enum=@n1.InstrumentType
    StartingInstrument:ESparksInstrumentType Enum=GetStartingInstrumentType!(self:Pilgrim Song Metadata Object Reference=@n1.SongMetadata, TrackType:EPilgrimTrackType Enum=@n1.InstrumentType) -> (ReturnValue:ESparksInstrumentType Enum=Guitar)
   ) -> (then)
  ]}
]
variables [
 Song List {type="array<PilgrimSongMetadata>", default []}
 SongEntries {type="array<WBP_PilgrimQuickplay_SetlistDisplaySongEntry_C>", default []}
]
widgets {
 info {inheritedWidgetCount=0, parentClass=&UserWidget, rootWidgetClass=&SizeBox, widgetCount=16}
 widgets [
  SizeBox_0 {
   widgetClassPath=&SizeBox
   widget=SizeBox_0
   properties {bOverride_Cursor=true, bOverride_HeightOverride=true, bOverride_WidthOverride=true, heightOverride=720, widthOverride=2450}
  }
  Background {
   widgetClassPath=&CommonBorder
   bIsVariable=true
   parent=SizeBox_0
   widget=Background
   properties {
    background {
     bIsDynamicallyLoaded=false
     drawAs=Image
     imageType=NoImage
     mirroring=NoMirror
     resourceName=None
     resourceObject=/SparksMaps/Effects/Screens/Materials/Environment/MI/MI_Sparks_AnimatedScreen_Env_UI_Blue1
     tiling=NoTile
     imageSize=<8 8>
     margin=<ltrb 0 0 0 0>
     outlineSettings {
      bUseBrushTransparency=false
      roundingType=HalfHeightRadius
      width=0
      color {colorUseRule=UseColor_Specified, specifiedColor=<rgba 0 0 0 0>}
      cornerRadii=<0 0 0 0>
     }
     tintColor {colorUseRule=UseColor_Specified, specifiedColor=<rgba 1 1 1 1>}
     uVRegion {bIsValid=false, max=<0 0>, min=<0 0>}
    }
    padding=<ltrb 20 20 20 20>
   }
   slot {class=SizeBoxSlot}
  }
  Overlay_62 {widgetClassPath=&Overlay, parent=Background, widget=Overlay_62, slot {class=BorderSlot, properties {padding=<ltrb 20 20 20 20>}}}
  VerticalBox {
   widgetClassPath=&VerticalBox
   parent=Overlay_62
   widget=VerticalBox
   slot {class=OverlaySlot, properties {horizontalAlignment=HAlign_Fill, verticalAlignment=VAlign_Fill}}
  }
  SetlistName {
   widgetClassPath=&CommonTextBlock
   bIsVariable=true
   parent=VerticalBox
   widget=SetlistName
   properties {
    clipping=OnDemand
    justification=Center
    scrollStyle=/PilgrimQuickplay/Widgets/ScrollStyles/TextScrollStyle_PilgrimQuickplay_Setlist.TextScrollStyle_PilgrimQuickplay_Setlist_C
    text=Band Setlist
    textTransformPolicy=ToUpper
    font {
     bForceMonospaced=false
     bMaterialIsStencil=false
     fontMaterial=null
     fontObject=&BurbankSmall
     letterSpacing=0
     monospacedWidth=1
     size=40
     skewAmount=0
     typefaceFontName=Black
     outlineSettings {
      bApplyOutlineToDropShadows=false
      bMiteredCorners=false
      bSeparateFillAlpha=false
      outlineMaterial=null
      outlineSize=0
      outlineColor=<rgba 0 0 0 1>
     }
    }
    shadowOffset=<0 0>
   }
   slot {class=VerticalBoxSlot, properties {verticalAlignment=VAlign_Center, padding=<ltrb 0 24 0 24>}}
  }
  SetlistGridPanel {
   widgetClassPath=/Script/UMG.UniformGridPanel
   bIsVariable=true
   parent=VerticalBox
   widget=SetlistGridPanel
   properties {bOverride_Cursor=true, slotPadding=<ltrb 8 8 8 8>}
   slot {class=VerticalBoxSlot, properties {size {sizeRule=Fill, value=1}}}
  }
  Song1 {
   widgetClassPath=&/WBP_PilgrimQuickplay_SetlistDisplaySongEntry
   bIsVariable=true
   parent=SetlistGridPanel
   widget=Song1
   properties {slotIdText=1, visibility=HitTestInvisible}
   slot {class=UniformGridSlot, properties {horizontalAlignment=HAlign_Fill, verticalAlignment=VAlign_Fill}}
  }
  Song2 {
   widgetClassPath=&WBP_PilgrimQuickplay_SetlistDisplaySongEntry
   bIsVariable=true
   parent=SetlistGridPanel
   widget=Song2
   properties {slotIdText=2, visibility=HitTestInvisible}
   slot {class=UniformGridSlot, properties {column=1, horizontalAlignment=HAlign_Fill, verticalAlignment=VAlign_Fill}}
  }
  Song3 {
   widgetClassPath=&WBP_PilgrimQuickplay_SetlistDisplaySongEntry
   bIsVariable=true
   parent=SetlistGridPanel
   widget=Song3
   properties {slotIdText=3, visibility=HitTestInvisible}
   slot {class=UniformGridSlot, properties {horizontalAlignment=HAlign_Fill, row=1, verticalAlignment=VAlign_Fill}}
  }
  Song4 {
   widgetClassPath=&WBP_PilgrimQuickplay_SetlistDisplaySongEntry
   bIsVariable=true
   parent=SetlistGridPanel
   widget=Song4
   properties {slotIdText=4, visibility=HitTestInvisible}
   slot {class=UniformGridSlot, properties {column=1, horizontalAlignment=HAlign_Fill, row=1, verticalAlignment=VAlign_Fill}}
  }
  Song5 {
   widgetClassPath=&WBP_PilgrimQuickplay_SetlistDisplaySongEntry
   bIsVariable=true
   parent=SetlistGridPanel
   widget=Song5
   properties {slotIdText=5, visibility=Collapsed}
   slot {class=UniformGridSlot, properties {horizontalAlignment=HAlign_Fill, row=2, verticalAlignment=VAlign_Fill}}
  }
  Song6 {
   widgetClassPath=&WBP_PilgrimQuickplay_SetlistDisplaySongEntry
   bIsVariable=true
   parent=SetlistGridPanel
   widget=Song6
   properties {slotIdText=6, visibility=Collapsed}
   slot {class=UniformGridSlot, properties {column=1, horizontalAlignment=HAlign_Fill, row=2, verticalAlignment=VAlign_Fill}}
  }
  NoSongsChosenLabel {
   widgetClassPath=&Overlay
   bIsVariable=true
   parent=Overlay_62
   widget=NoSongsChosenLabel
   properties {visibility=Visible}
   slot {class=OverlaySlot, properties {horizontalAlignment=HAlign_Fill, verticalAlignment=VAlign_Fill}}
  }
  BG {
   widgetClassPath=&CommonBorder
   bIsVariable=true
   parent=NoSongsChosenLabel
   widget=BG
   properties {
    background {
     bIsDynamicallyLoaded=false
     drawAs=Image
     imageType=NoImage
     mirroring=NoMirror
     resourceName=None
     resourceObject=&MI_Sparks_AnimatedScreen_Env_UI_Blue1
     tiling=NoTile
     imageSize=<8 8>
     margin=<ltrb 0 0 0 0>
     outlineSettings {
      bUseBrushTransparency=false
      roundingType=HalfHeightRadius
      width=0
      color {colorUseRule=UseColor_Specified, specifiedColor=<rgba 0 0 0 0>}
      cornerRadii=<0 0 0 0>
     }
     tintColor {colorUseRule=UseColor_Specified, specifiedColor=<rgba 1 1 1 1>}
     uVRegion {bIsValid=false, max=<0 0>, min=<0 0>}
    }
    padding=<ltrb 0 0 0 0>
   }
   slot {class=OverlaySlot, properties {horizontalAlignment=HAlign_Fill, verticalAlignment=VAlign_Fill}}
  }
  Border {
   widgetClassPath=&CommonBorder
   parent=NoSongsChosenLabel
   widget=Border
   properties {
    horizontalAlignment=HAlign_Center
    verticalAlignment=VAlign_Center
    background {
     bIsDynamicallyLoaded=false
     drawAs=NoDrawType
     imageType=NoImage
     mirroring=NoMirror
     resourceName=None
     resourceObject=/Game/UI/GreyBox/white
     tiling=NoTile
     imageSize=<8 8>
     margin=<ltrb 0 0 0 0>
     outlineSettings {
      bUseBrushTransparency=false
      roundingType=HalfHeightRadius
      width=0
      color {colorUseRule=UseColor_Specified, specifiedColor=<rgba 0 0 0 0>}
      cornerRadii=<0 0 0 0>
     }
     tintColor {colorUseRule=UseColor_Specified, specifiedColor=<rgba 1 1 1 1>}
     uVRegion {bIsValid=false, max=<0 0>, min=<0 0>}
    }
    padding=<ltrb 0 0 0 0>
   }
   slot {class=OverlaySlot, properties {horizontalAlignment=HAlign_Fill, verticalAlignment=VAlign_Fill}}
  }
  NoSongsText {
   widgetClassPath=&CommonTextBlock
   parent=Border
   widget=NoSongsText
   properties {
    justification=Center
    text=Choose a Song
    textTransformPolicy=ToUpper
    font {
     bForceMonospaced=false
     bMaterialIsStencil=false
     fontMaterial=null
     fontObject=&HeadingNow
     letterSpacing=0
     monospacedWidth=1
     size=75
     skewAmount=0
     typefaceFontName=76 Bold
     outlineSettings {
      bApplyOutlineToDropShadows=false
      bMiteredCorners=false
      bSeparateFillAlpha=false
      outlineMaterial=null
      outlineSize=0
      outlineColor=<rgba 0 0 0 1>
     }
    }
    shadowOffset=<0 0>
   }
   slot {class=BorderSlot, properties {horizontalAlignment=HAlign_Center, verticalAlignment=VAlign_Center, padding=<ltrb 0 0 0 0>}}
  }
 ]
}
