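# Pilgrim Quickplay widget that shows a song as album art with its title and artist overlaid.
# SetSongMetadata stores the song metadata, sets the artist and title text from it, then lazy-loads the cover into the AlbumArt image, which has a disc texture as its default brush.
# A size-capped, scaled overlay stacks the art, a dark bottom gradient and bottom-placed text blocks that use a setlist scroll style.
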WidgetBlueprint /PilgrimQuickplay/Widgets/SetlistAndSongDisplay/WBP_PilgrimQuickplay_FullArtSongDisplay
graphs [
 EventGraph {nodes [
   Event PreConstruct()
   Event Construct()
   Event Tick()
  ]}
 SetSongMetadata {nodes [
   n1 = SetSongMetadata!() -> (then, NewSongMetadata:Pilgrim Song Metadata Object Reference) {is=title K2Node_FunctionEntry}
   > n2 = Set SongMetadata(SongMetadata:Pilgrim Song Metadata Object Reference=@n1) -> (Output_Get:Pilgrim Song Metadata Object Reference)
   > n3 = Sequence!(execute) -> (then_0, then_1, then_2:Exec)
   > "SetText (Text)"(
    self=Get SongArtist() -> (SongArtist:Common Text Object Reference)
    InText=GetArtistText!(self:Sparks Song Data Object Reference=@n2) -> (ReturnValue:Text)
   )
   "SetText (Text)"(
    execute=@n3.then_1
    self=Get SongTitle() -> (SongTitle:Common Text Object Reference)
    InText=GetTitleText!(self:Sparks Song Data Object Reference=@n2) -> (ReturnValue:Text)
   )
   LazyLoadAlbumArt!(
    execute=@n3.then_2
    Song Metadata:Pilgrim Song Metadata Object Reference=@n2
    Lazy Image:Fort Lazy Image Object Reference=Get AlbumArt() -> (AlbumArt:Fort Lazy Image Object Reference)
   ) -> (then)
  ]}
]
variables [SongMetadata {type=PilgrimSongMetadata, default=null}]
widgets {
 info {inheritedWidgetCount=0, parentClass=&UserWidget, rootWidgetClass=&Overlay, widgetCount=11}
 widgets [
  Overlay_26 {widgetClassPath=&Overlay, widget=Overlay_26}
  ScaleBox_68 {
   widgetClassPath=&ScaleBox
   parent=Overlay_26
   widget=ScaleBox_68
   slot {class=OverlaySlot, properties {horizontalAlignment=HAlign_Center, verticalAlignment=VAlign_Center}}
  }
  SizeBox_104 {
   widgetClassPath=&SizeBox
   parent=ScaleBox_68
   widget=SizeBox_104
   properties {bOverride_MaxDesiredHeight=true, bOverride_MaxDesiredWidth=true, maxDesiredHeight=400, maxDesiredWidth=400}
   slot {class=ScaleBoxSlot}
  }
  Overlay_122 {
   widgetClassPath=&Overlay
   parent=SizeBox_104
   widget=Overlay_122
   slot {class=SizeBoxSlot, properties {horizontalAlignment=HAlign_Center, verticalAlignment=VAlign_Center}}
  }
  AlbumArt {
   widgetClassPath=&FortLazyImage
   bIsVariable=true
   parent=Overlay_122
   widget=AlbumArt
   properties {brush {
     bIsDynamicallyLoaded=false
     drawAs=Image
     imageType=NoImage
     mirroring=NoMirror
     resourceName=None
     resourceObject=/PilgrimCore/Widgets/Textures/T_HMX_Disc_01_D
     tiling=NoTile
     imageSize=<2048 2048>
     margin=<ltrb 0 0 0 0>
     outlineSettings {
      bUseBrushTransparency=false
      roundingType=HalfHeightRadius
      width=0
      color {colorUseRule=UseColor_Specified, specifiedColor=<rgba 0 0 0 0>}
      cornerRadii=<0 0 0 0>
     }
     tintColor {colorUseRule=UseColor_Specified, specifiedColor=<rgba 1 1 1 1>}
     uVRegion {bIsValid=false, max=<0 0>, min=<0 0>}
    }}
   slot {class=OverlaySlot}
  }
  Gradient {
   widgetClassPath=&Image
   parent=Overlay_122
   widget=Gradient
   properties {
    brush {
     bIsDynamicallyLoaded=false
     drawAs=Image
     imageType=NoImage
     mirroring=NoMirror
     resourceName=None
     resourceObject=/FortUILibrary/MaterialInstances/MI_UI_LinearGradientBottom
     tiling=NoTile
     imageSize=<32 32>
     margin=<ltrb 0 0 0 0>
     outlineSettings {
      bUseBrushTransparency=false
      roundingType=HalfHeightRadius
      width=0
      color {colorUseRule=UseColor_Specified, specifiedColor=<rgba 0 0 0 0>}
      cornerRadii=<0 0 0 0>
     }
     tintColor {colorUseRule=UseColor_Specified, specifiedColor=<rgba 1 1 1 1>}
     uVRegion {bIsValid=false, max=<0 0>, min=<0 0>}
    }
    colorAndOpacity=<rgba 0 0 0 1>
   }
   slot {class=OverlaySlot, properties {horizontalAlignment=HAlign_Fill, verticalAlignment=VAlign_Fill}}
  }
  VerticalBox_34 {
   widgetClassPath=&VerticalBox
   parent=Overlay_122
   widget=VerticalBox_34
   slot {class=OverlaySlot, properties {horizontalAlignment=HAlign_Fill, verticalAlignment=VAlign_Fill}}
  }
  Spacer_65 {
   widgetClassPath=&Spacer
   parent=VerticalBox_34
   widget=Spacer_65
   slot {class=VerticalBoxSlot, properties {size {sizeRule=Fill, value=0.85}}}
  }
  SongTitle {
   widgetClassPath=&CommonTextBlock
   bIsVariable=true
   parent=VerticalBox_34
   widget=SongTitle
   properties {
    clipping=OnDemand
    justification=Center
    scrollStyle=/PilgrimQuickplay/Widgets/ScrollStyles/TextScrollStyle_PilgrimQuickplay_Setlist.TextScrollStyle_PilgrimQuickplay_Setlist_C
    text=SONG TITLE
    font {
     bForceMonospaced=false
     bMaterialIsStencil=false
     fontMaterial=null
     fontObject=&BurbankSmall
     letterSpacing=0
     monospacedWidth=1
     size=40
     skewAmount=0
     typefaceFontName=Bold
     outlineSettings {
      bApplyOutlineToDropShadows=false
      bMiteredCorners=false
      bSeparateFillAlpha=false
      outlineMaterial=null
      outlineSize=0
      outlineColor=<rgba 0 0 0 1>
     }
    }
    shadowOffset=<0 0>
   }
   slot {class=VerticalBoxSlot, properties {verticalAlignment=VAlign_Bottom, padding=<ltrb 10 0 10 0>, size {sizeRule=Automatic, value=0.15}}}
  }
  SongArtist {
   widgetClassPath=&CommonTextBlock
   bIsVariable=true
   parent=VerticalBox_34
   widget=SongArtist
   properties {
    clipping=OnDemand
    justification=Center
    scrollStyle=&TextScrollStyle_PilgrimQuickplay_Setlist_C
    text=ARTIST
    colorAndOpacity {colorUseRule=UseColor_Specified, specifiedColor=<rgba 0.1946 0.5972 0.9473 1>}
    font {
     bForceMonospaced=false
     bMaterialIsStencil=false
     fontMaterial=null
     fontObject=&NotoSans
     letterSpacing=0
     monospacedWidth=1
     size=30
     skewAmount=0
     typefaceFontName=NotoSans-Bold
     outlineSettings {
      bApplyOutlineToDropShadows=false
      bMiteredCorners=false
      bSeparateFillAlpha=false
      outlineMaterial=null
      outlineSize=0
      outlineColor=<rgba 0 0 0 1>
     }
    }
    shadowOffset=<0 0>
   }
   slot {class=VerticalBoxSlot, properties {verticalAlignment=VAlign_Top, padding=<ltrb 10 0 10 0>, size {sizeRule=Automatic, value=0.1}}}
  }
  Spacer {
   widgetClassPath=&Spacer
   parent=VerticalBox_34
   widget=Spacer
   slot {class=VerticalBoxSlot, properties {size {sizeRule=Fill, value=0.1}}}
  }
 ]
}
