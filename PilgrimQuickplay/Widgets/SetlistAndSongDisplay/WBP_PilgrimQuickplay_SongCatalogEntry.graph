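# CommonButton entry for a song in the Pilgrim Quickplay catalog, wrapping a full-art song display that receives the song metadata.
# Set Song Display Status switches between unselected, selected and disabled, with the disabled state labeled "Added By" plus the choosing player's name.
# Focusing an entry in that disabled state shows a tinted overlay and an "Already Added" message.
# That message and the externally callable max-songs message share the FadeMessage opacity animation.
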WidgetBlueprint /PilgrimQuickplay/Widgets/SetlistAndSongDisplay/WBP_PilgrimQuickplay_SongCatalogEntry
animations [FadeMessage {
  binding_count=1
  length_seconds=2
  object=&self:FadeMessage
  sequence {bindings [MessageOverlay {
     id=#1
     possessed_class=Overlay
     tracks [{
       type=MovieSceneFloatTrack
       display_name=Render Opacity
       sections [{type=MovieSceneFloatSection, end_frame=40, start_frame=0, channels [None {type=MovieSceneScriptingFloatChannel, keys [[0,1],[40,0]]}]}]
      }]
    }]}
 }]
graphs [
 EventGraph {nodes [
   n1 = SetDisabledFocusOverlayVisibility!() -> (OutputDelegate, then, Visibility:ESlateVisibility Enum=Visible) {is=title K2Node_CustomEvent}
   > Branch(Condition="Equal (Enum)"(
     A:ESongDisplayStatus Enum=Get DisplayStatus() -> (DisplayStatus:ESongDisplayStatus Enum=NewEnumerator2)
     B:ESongDisplayStatus Enum=NewEnumerator1
    ))
   > SetVisibility(self=Get DisabledOverlay() -> (DisabledOverlay:Border Object Reference), InVisibility=@n1.Visibility)
   > Branch(Condition="Equal (Enum)"(A:ESlateVisibility Enum=@n1.Visibility, B:ESlateVisibility Enum=SelfHitTestInvisible))
   > ShowAlreadySelectedMessage()
   Event Construct()
   > SetRenderOpacity(self=Get MessageOverlay() -> (MessageOverlay:Overlay Object Reference), InOpacity=0)
   Event On Focused() {is=title K2Node_Event}
   > SetDisabledFocusOverlayVisibility!(execute, self, Visibility:ESlateVisibility Enum=SelfHitTestInvisible) -> (then) {is=title K2Node_CallFunction}
   Event On Unfocused() {is=title K2Node_Event}
   > SetDisabledFocusOverlayVisibility(Visibility=Hidden)
   Event PreConstruct()
   Event Tick()
  ]}
 GetSongMetadata {nodes [
   GetSongMetadata() {is=title K2Node_FunctionEntry}
   > Return Node!(
    execute
    SongMetadata:Pilgrim Song Metadata Object Reference=Get SongMetadata!(self:WBP Pilgrim Quickplay Full Art Song Display Object Reference=Get WBP_PilgrimQuickplay_FullArtSongDisplay() -> (WBP_PilgrimQuickplay_FullArtSongDisplay:WBP Pilgrim Quickplay Full Art Song Display Object Reference)) -> (SongMetadata:Pilgrim Song Metadata Object Reference)
   )
  ]}
 PlayMessageAnimation {nodes [
   n1 = PlayMessageAnimation!() -> (then, MessageText:Text) {is=title K2Node_FunctionEntry}
   > "SetText (Text)"(self=Get MessageText() -> (MessageText:Common Text Object Reference), InText=@n1)
   > PlayAnimation(InAnimation=Get FadeMessage() -> (FadeMessage:Widget Animation Object Reference))
  ]}
 Set Song Display Status {nodes [
   n1 = Set Song Display Status!() -> (then, SongStatus:Song Display Status Struct Structure) {is=title K2Node_FunctionEntry}
   n2 = Break Song Display Status Struct!(SongDisplayStatusStruct:"Song Display Status Struct Structure (by ref)"=@n1) -> (
    SongShortName_19_#2:Name
    DisplayStatus_14_#3:ESongDisplayStatus Enum
    SongChooserPlayerState_17_#4:Player State Object Reference
   ) {is=title K2Node_BreakStruct}
   > n3 = Set DisplayStatus(DisplayStatus:ESongDisplayStatus Enum=@n2.DisplayStatus_14_#3) -> (Output_Get:ESongDisplayStatus Enum=NewEnumerator2)
   > n4 = Switch on ESongDisplayStatus!(execute, Selection:ESongDisplayStatus Enum=@n3) -> (NewEnumerator2:Exec, NewEnumerator0:Exec, NewEnumerator1:Exec) {is=title K2Node_SwitchEnum}
   n5 = SetSelectedInternal!(execute=@n4.NewEnumerator0, self:Common Button Base Object Reference, bInSelected:bool=true, bAllowSound:bool=true, bBroadcast:bool=true) -> (then)
   SetIsEnabled(execute=[@n7.then, @n5.then], bInIsEnabled=true)
   n6 = Get AlreadyAddedLabel() -> (AlreadyAddedLabel:Vertical Box Object Reference)
   > SetVisibility(self=@n6, InVisibility=Hidden)
   SetIsEnabled(execute=@n4.NewEnumerator1, bInIsEnabled=false)
   > SetVisibility(self=@n6, InVisibility=HitTestInvisible)
   > "SetText (Text)"(
    self=Get AddedByPlayerText() -> (AddedByPlayerText:Common Text Object Reference)
    InText=Format Text!(
     Format=Get AddedByPlayer() -> (AddedByPlayer:Text)
     0:Text=MakeLiteralText(Value="INVTEXT(\"\\r\\n\")")
     1:String=GetPlayerName!(self:Player State Object Reference=@n2.SongChooserPlayerState_17_#4) -> (ReturnValue:String)
    ) -> (Result)
   )
   n7 = SetSelectedInternal(execute=@n4.NewEnumerator2, bInSelected=false, bAllowSound=true, bBroadcast=true)
  ]}
 SetSongMetadata {nodes [
   n1 = SetSongMetadata!() -> (then, NewSongMetadata:Pilgrim Song Metadata Object Reference) {is=title K2Node_FunctionEntry}
   > SetSongMetadata!(
    execute
    self:WBP Pilgrim Quickplay Full Art Song Display Object Reference=Get WBP_PilgrimQuickplay_FullArtSongDisplay()
    NewSongMetadata:Pilgrim Song Metadata Object Reference=@n1
   ) -> (then) {is=title K2Node_CallFunction}
  ]}
 ShowAlreadySelectedMessage {nodes [
   ShowAlreadySelectedMessage() {is=title K2Node_FunctionEntry}
   > PlayMessageAnimation!(execute, self, MessageText:Text=Get AlreadySelected() -> (AlreadySelected:Text)) -> (then) {is=title K2Node_CallFunction}
  ]}
 ShowMaxSongsMessage {nodes [
   ShowMaxSongsMessage() {is=title K2Node_FunctionEntry}
   > PlayMessageAnimation(MessageText=Get AtMaxSongs() -> (AtMaxSongs:Text))
  ]}
]
variables [
 AddedByPlayer {type=string, default="Added By{0}{1}"}
 AlreadySelected {type=string, default=Already Added}
 AtMaxSongs {type=string, default=max songs chosen}
 DisplayStatus {type=ESongDisplayStatus, default=None}
 PrevDisplayStatus {type=ESongDisplayStatus, default=None}
]
widgets {
 info {inheritedWidgetCount=1, parentClass=&CommonButtonBase, rootWidgetClass=&CanvasPanel, widgetCount=15}
 widgets [
  CanvasPanel_2 {widgetClassPath=&CanvasPanel, widget=CanvasPanel_2}
  Overlay_1 {
   widgetClassPath=&Overlay
   parent=CanvasPanel_2
   widget=Overlay_1
   slot {class=CanvasPanelSlot, properties {layoutData {alignment=<0 0>, anchors {maximum=<1 1>, minimum=<0 0>}, offsets=<ltrb 0 0 0 0>}}}
  }
  WBP_PilgrimQuickplay_FullArtSongDisplay {
   widgetClassPath=&/WBP_PilgrimQuickplay_FullArtSongDisplay
   bIsVariable=true
   parent=Overlay_1
   widget=WBP_PilgrimQuickplay_FullArtSongDisplay
   slot {class=OverlaySlot, properties {horizontalAlignment=HAlign_Center, verticalAlignment=VAlign_Center, padding=<ltrb 25 25 25 25>}}
  }
  DisabledOverlay {
   widgetClassPath=&Border
   bIsVariable=true
   parent=Overlay_1
   widget=DisabledOverlay
   properties {visibility=Hidden, brushColor=<rgba 0 1 0.6654 0.3>}
   slot {class=OverlaySlot, properties {horizontalAlignment=HAlign_Fill, verticalAlignment=VAlign_Fill}}
  }
  AlreadyAddedLabel {
   widgetClassPath=&VerticalBox
   bIsVariable=true
   parent=Overlay_1
   widget=AlreadyAddedLabel
   properties {visibility=Hidden}
   slot {class=OverlaySlot, properties {horizontalAlignment=HAlign_Fill, verticalAlignment=VAlign_Fill}}
  }
  Spacer {
   widgetClassPath=&Spacer
   parent=AlreadyAddedLabel
   widget=Spacer
   slot {class=VerticalBoxSlot, properties {size {sizeRule=Fill, value=0.25}}}
  }
  YellowBG {
   widgetClassPath=&Border
   parent=AlreadyAddedLabel
   widget=YellowBG
   properties {
    background {
     bIsDynamicallyLoaded=false
     drawAs=Image
     imageType=NoImage
     mirroring=NoMirror
     resourceName=None
     resourceObject=/Game/UI/Frontend/Showdown/Materials/M_UI_SkewBox_LButton_Yellow
     tiling=NoTile
     imageSize=<32 32>
     margin=<ltrb 0 0 0 0>
     outlineSettings {
      bUseBrushTransparency=false
      roundingType=HalfHeightRadius
      width=0
      color {colorUseRule=UseColor_Specified, specifiedColor=<rgba 0 0 0 0>}
      cornerRadii=<0 0 0 0>
     }
     tintColor {colorUseRule=UseColor_Specified, specifiedColor=<rgba 1 1 1 1>}
     uVRegion {bIsValid=false, max=<0 0>, min=<0 0>}
    }
    desiredSizeScale=<50 50>
    padding=<ltrb 15 15 15 15>
   }
   slot {
    class=VerticalBoxSlot
    properties {horizontalAlignment=HAlign_Center, verticalAlignment=VAlign_Center, padding=<ltrb 20 0 20 0>, size {sizeRule=Fill, value=1}}
   }
  }
  ScaleBox_324 {widgetClassPath=&ScaleBox, parent=YellowBG, widget=ScaleBox_324, slot {class=BorderSlot, properties {padding=<ltrb 15 15 15 15>}}}
  AddedByPlayerText {
   widgetClassPath=&CommonTextBlock
   bIsVariable=true
   parent=ScaleBox_324
   widget=AddedByPlayerText
   properties {
    justification=Center
    style=/FortUILibrary/Styles/Typography/Foundation/TextStyle-BurbankSmall-M-Black.TextStyle-BurbankSmall-M-Black_C
    text="Added by\r\n{0}"
    textTransformPolicy=ToUpper
    colorAndOpacity {colorUseRule=UseColor_Specified, specifiedColor=<rgba 0 0 0 1>}
    font {
     bForceMonospaced=false
     bMaterialIsStencil=false
     fontMaterial=null
     fontObject=&BurbankSmall
     letterSpacing=0
     monospacedWidth=1
     size=24
     skewAmount=0
     typefaceFontName=Black
     outlineSettings {
      bApplyOutlineToDropShadows=false
      bMiteredCorners=false
      bSeparateFillAlpha=false
      outlineMaterial=null
      outlineSize=0
      outlineColor=<rgba 0 0 0 1>
     }
    }
    shadowOffset=<0 0>
   }
   slot {class=ScaleBoxSlot}
  }
  Spacer_70 {
   widgetClassPath=&Spacer
   parent=AlreadyAddedLabel
   widget=Spacer_70
   slot {class=VerticalBoxSlot, properties {size {sizeRule=Fill, value=1}}}
  }
  MessageOverlay {
   widgetClassPath=&Overlay
   bIsVariable=true
   parent=Overlay_1
   widget=MessageOverlay
   slot {class=OverlaySlot, properties {horizontalAlignment=HAlign_Fill, verticalAlignment=VAlign_Fill, padding=<ltrb 50 50 50 50>}}
  }
  Border_80 {
   widgetClassPath=&Border
   parent=MessageOverlay
   widget=Border_80
   properties {
    background {
     bIsDynamicallyLoaded=false
     drawAs=Box
     imageType=NoImage
     mirroring=NoMirror
     resourceName=None
     resourceObject=null
     tiling=NoTile
     imageSize=<32 32>
     margin=<ltrb 0 0 0 0>
     outlineSettings {
      bUseBrushTransparency=false
      roundingType=FixedRadius
      width=0
      color {colorUseRule=UseColor_Specified, specifiedColor=<rgba 1 1 1 1>}
      cornerRadii=<20 20 20 20>
     }
     tintColor {colorUseRule=UseColor_Specified, specifiedColor=<rgba 1 1 1 1>}
     uVRegion {bIsValid=false, max=<0 0>, min=<0 0>}
    }
    brushColor=<rgba 0.0003 0.0762 0.4125 1>
   }
   slot {class=OverlaySlot, properties {horizontalAlignment=HAlign_Fill, verticalAlignment=VAlign_Fill}}
  }
  MessageText {
   widgetClassPath=&CommonTextBlock
   bIsVariable=true
   parent=MessageOverlay
   widget=MessageText
   properties {
    autoWrapText=true
    justification=Center
    style=/FortUILibrary/Styles/Typography/Foundation/TextStyle_BurbankSmall_M_White.TextStyle_BurbankSmall_M_White_C
    text=MESSAGE TEXT
    textTransformPolicy=ToUpper
    font {
     bForceMonospaced=false
     bMaterialIsStencil=false
     fontMaterial=null
     fontObject=&BurbankSmall
     letterSpacing=0
     monospacedWidth=1
     size=24
     skewAmount=0
     typefaceFontName=Black
     outlineSettings {
      bApplyOutlineToDropShadows=false
      bMiteredCorners=false
      bSeparateFillAlpha=false
      outlineMaterial=null
      outlineSize=0
      outlineColor=<rgba 0 0 0 1>
     }
    }
    shadowOffset=<0 0>
   }
   slot {class=OverlaySlot, properties {horizontalAlignment=HAlign_Center, verticalAlignment=VAlign_Center}}
  }
  WBP_PilgrimQuickplay_QuestMarker {
   widgetClassPath=/PilgrimQuickplay/Widgets/Quests/WBP_PilgrimQuickplay_QuestMarker
   bIsVariable=true
   parent=CanvasPanel_2
   widget=WBP_PilgrimQuickplay_QuestMarker
   properties {visibility=Collapsed}
   slot {
    class=CanvasPanelSlot
    properties {zOrder=1, layoutData {alignment=<0 0>, anchors {maximum=<0 0>, minimum=<0 0>}, offsets=<ltrb -20 -20 100 100>}}
   }
  }
  InputActionWidget {widgetClassPath=&CommonActionWidget, bInherited=true, widget=null}
 ]
}
