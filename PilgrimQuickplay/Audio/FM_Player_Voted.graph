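# Sound cue in PilgrimQuickplay that, by its name, plays when a player has voted.
# A single wave player outputs the StreakGained wave borrowed from the SparksMaps audio content.
# It plays at reduced volume under the in-game SFX sound class, with a very large max distance.
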SoundCue /PilgrimQuickplay/Audio/FM_Player_Voted
root=WavePlayer_0
nodes [{id=WavePlayer_0, properties {soundWaveAssetPtr=/SparksMaps/Audio/Wavs/StreakGained}}]
properties {maxDistance=2097152, soundClassObject=&Fort_SFX_InGame, volumeMultiplier=0.45}
