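# Translucent UI material that draws a small instrument-type icon, optionally over a circular background, for Pilgrim Quickplay widgets.
# The Icon scalar blends through channels of one packed texture via chained lerps: B at 0 and 1, A at 2, G at 3.
# A smoothstepped MF_UI_Circle SDF forms the background disc.
# Static switches add the disc to the icon mask and combine its color with the tinted icon.
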Material /PilgrimQuickplay/Widgets/Materials/M_UI_Icon_Instrument_Type_Sm
graph {
 properties {blendMode=BLEND_Translucent, materialDomain=MD_UI, pixelDepthOffsetMode=PDOM_Legacy, usedShadingModels=Default Lit}
 nodes [
  n1 = OneMinus(None=SmoothStep(Max=Constant() {r=0.38}, Value=MaterialFunctionCall() {materialFunction=&MF_UI_Circle}.SDF) {constMax=0.05, constMin=0.35})
  n2 = TextureSample() {samplerType=SAMPLERTYPE_LinearColor, texture=/PilgrimQuickplay/Widgets/Textures/T_Icon_Instrument_Type}
  n3 = ScalarParameter() {defaultValue=1, parameterName=Icon}
  n4 = Saturate(None=@n3)
  n5 = Subtract(A=@n3)
  n6 = LinearInterpolate(
   A=LinearInterpolate(A=LinearInterpolate(A=@n2.B, B=@n2.B, Alpha=@n4), B=@n2.A, Alpha=Saturate(None=@n5))
   B=@n2.G
   Alpha=Saturate(None=Subtract(A=@n3) {constB=2})
  )
  StaticSwitchParameter(True=Add(A=@n1, B=@n6), False=@n6) {defaultValue=true, parameterName=Bg Toggle}
  n7 = VectorParameter() {parameterName=Bg Color, defaultValue=<rgba 0 0 0 1>}
  n8 = LinearInterpolate(B=VectorParameter() {parameterName=IconColor, defaultValue=<rgba 1 1 1 1>}.RGB, Alpha=@n6)
  StaticSwitchParameter(True=Max(A=LinearInterpolate(B=@n7.RGB, Alpha=@n1), B=@n8), False=@n8) {defaultValue=true, parameterName=BgColor}
 ]
}
