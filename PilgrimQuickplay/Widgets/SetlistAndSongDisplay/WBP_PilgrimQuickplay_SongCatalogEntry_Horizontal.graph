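# Horizontal song catalog list entry for Pilgrim Quickplay, showing art, title, artist, intensities, personal best, quest marker and spotlight instrument markers.
# Its item object supplies metadata and display status, toggling setlist indicators and the adding player's name and avatar.
# Update events flash fading "Already Added" or max-songs messages.
# Selection plays Selected_Animated forward to enlarge art and reveal details, and deselection or release reverses it.
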WidgetBlueprint /PilgrimQuickplay/Widgets/SetlistAndSongDisplay/WBP_PilgrimQuickplay_SongCatalogEntry_Horizontal
animations [
 FadeMessage {
  binding_count=1
  length_seconds=2
  object=&self:FadeMessage
  sequence {bindings [MessageOverlay {
     id=#1
     possessed_class=Overlay
     tracks [{
       type=MovieSceneFloatTrack
       display_name=Render Opacity
       sections [{type=MovieSceneFloatSection, end_frame=40, start_frame=0, channels [None {type=MovieSceneScriptingFloatChannel, keys [[0,1],[40,0]]}]}]
      }]
    }]}
 }
 Selected {
  binding_count=9
  length_seconds=0.3
  object=&self:Selected
  sequence {bindings [
    QuestMarker {
     id=#2
     possessed_class=Image
     tracks [{
       type=MovieSceneColorTrack
       display_name=Color and Opacity
       sections [{
         type=MovieSceneColorSection
         end_frame=6
         start_frame=0
         channels [
          Color.R {type=MovieSceneScriptingFloatChannel, const [0,2]}
          Color.G {type=MovieSceneScriptingFloatChannel, const [0,2]}
          Color.B {type=MovieSceneScriptingFloatChannel, const [0,2]}
          Color.A {type=MovieSceneScriptingFloatChannel, const [1,2]}
         ]
        }]
      }]
    }
    SongArtist {
     id=#3
     possessed_class=CommonTextBlock
     tracks [{
       type=MovieSceneColorTrack
       display_name=Color and Opacity
       sections [{
         type=MovieSceneColorSection
         end_frame=6
         start_frame=0
         channels [
          Color.R {type=MovieSceneScriptingFloatChannel, const [0,2]}
          Color.G {type=MovieSceneScriptingFloatChannel, const [0,2]}
          Color.B {type=MovieSceneScriptingFloatChannel, const [0,2]}
          Color.A {type=MovieSceneScriptingFloatChannel, const [1,2]}
         ]
        }]
      }]
    }
    IntensitiesScaleBox {
     id=#4
     possessed_class=ScaleBox
     tracks [{
       type=MovieSceneFloatTrack
       display_name=User Specified Scale
       sections [{type=MovieSceneFloatSection, end_frame=6, start_frame=0, channels [None {type=MovieSceneScriptingFloatChannel, const [0.75,2]}]}]
      }]
    }
    Text_Score {
     id=#5
     possessed_class=CommonTextBlock
     tracks [{
       type=MovieSceneColorTrack
       display_name=Color and Opacity
       sections [{
         type=MovieSceneColorSection
         end_frame=6
         start_frame=0
         channels [
          Color.R {type=MovieSceneScriptingFloatChannel, const [0,2]}
          Color.G {type=MovieSceneScriptingFloatChannel, const [0,2]}
          Color.B {type=MovieSceneScriptingFloatChannel, const [0,2]}
          Color.A {type=MovieSceneScriptingFloatChannel, const [1,2]}
         ]
        }]
      }]
    }
    SongTitle {
     id=#6
     possessed_class=CommonTextBlock
     tracks [{
       type=MovieSceneColorTrack
       display_name=Color and Opacity
       sections [{
         type=MovieSceneColorSection
         end_frame=6
         start_frame=0
         channels [
          Color.R {type=MovieSceneScriptingFloatChannel, const [0,2]}
          Color.G {type=MovieSceneScriptingFloatChannel, const [0,2]}
          Color.B {type=MovieSceneScriptingFloatChannel, const [0,2]}
          Color.A {type=MovieSceneScriptingFloatChannel}
         ]
        }]
      }]
    }
    IntensitiesBorder {
     id=#7
     possessed_class=CommonBorder
     tracks [{
       type=MovieSceneColorTrack
       display_name=Content Color and Opacity
       sections [{
         type=MovieSceneColorSection
         end_frame=6
         start_frame=0
         channels [
          Color.R {type=MovieSceneScriptingFloatChannel}
          Color.G {type=MovieSceneScriptingFloatChannel}
          Color.B {type=MovieSceneScriptingFloatChannel}
          Color.A {type=MovieSceneScriptingFloatChannel, const [1,2]}
         ]
        }]
      }]
    }
    AlbumArtSizeBox {
     id=#8
     possessed_class=SizeBox
     tracks [
      {
       type=MovieSceneFloatTrack
       display_name=Width Override
       sections [{type=MovieSceneFloatSection, end_frame=6, start_frame=0, channels [None {type=MovieSceneScriptingFloatChannel, const [250,2]}]}]
      }
      {
       type=MovieSceneFloatTrack
       display_name=Height Override
       sections [{type=MovieSceneFloatSection, end_frame=6, start_frame=0, channels [None {type=MovieSceneScriptingFloatChannel, const [250,2]}]}]
      }
     ]
    }
    ScoreDetailsScaleBox {
     id=#9
     possessed_class=ScaleBox
     tracks [
      {
       type=MovieSceneFloatTrack
       display_name=User Specified Scale
       sections [{type=MovieSceneFloatSection, end_frame=6, start_frame=0, channels [None {type=MovieSceneScriptingFloatChannel, const [1,2]}]}]
      }
      {
       type=MovieSceneFloatTrack
       display_name=Render Opacity
       sections [{type=MovieSceneFloatSection, end_frame=6, start_frame=0, channels [None {type=MovieSceneScriptingFloatChannel, const [1,2]}]}]
      }
     ]
    }
    Text_Accuracy {
     id=#10
     possessed_class=CommonTextBlock
     tracks [{
       type=MovieSceneColorTrack
       display_name=Color and Opacity
       sections [{
         type=MovieSceneColorSection
         end_frame=6
         start_frame=0
         channels [
          Color.R {type=MovieSceneScriptingFloatChannel, const [0,2]}
          Color.G {type=MovieSceneScriptingFloatChannel, const [0,2]}
          Color.B {type=MovieSceneScriptingFloatChannel, const [0,2]}
          Color.A {type=MovieSceneScriptingFloatChannel, const [1,2]}
         ]
        }]
      }]
    }
   ]}
 }
 Selected_Animated {
  binding_count=9
  length_seconds=0.3
  object=&self:Selected_Animated
  sequence {bindings [
    QuestMarker {
     id=#2
     possessed_class=Image
     tracks [{
       type=MovieSceneColorTrack
       display_name=Color and Opacity
       sections [{
         type=MovieSceneColorSection
         end_frame=6
         start_frame=0
         channels [
          Color.R {type=MovieSceneScriptingFloatChannel, keys [[0,1],[6,0]]}
          Color.G {type=MovieSceneScriptingFloatChannel, keys [[0,1],[6,0]]}
          Color.B {type=MovieSceneScriptingFloatChannel, keys [[0,1],[6,0]]}
          Color.A {type=MovieSceneScriptingFloatChannel, const [1,2]}
         ]
        }]
      }]
    }
    SongArtist {
     id=#3
     possessed_class=CommonTextBlock
     tracks [{
       type=MovieSceneColorTrack
       display_name=Color and Opacity
       sections [{
         type=MovieSceneColorSection
         end_frame=6
         start_frame=0
         channels [
          Color.R {type=MovieSceneScriptingFloatChannel, keys [[0,1],[6,0]]}
          Color.G {type=MovieSceneScriptingFloatChannel, keys [[0,1],[6,0]]}
          Color.B {type=MovieSceneScriptingFloatChannel, keys [[0,1],[6,0]]}
          Color.A {type=MovieSceneScriptingFloatChannel, const [1,2]}
         ]
        }]
      }]
    }
    IntensitiesScaleBox {
     id=#4
     possessed_class=ScaleBox
     tracks [{
       type=MovieSceneFloatTrack
       display_name=User Specified Scale
       sections [{type=MovieSceneFloatSection, end_frame=6, start_frame=0, channels [None {type=MovieSceneScriptingFloatChannel, keys [[0,0],[6,0.75]]}]}]
      }]
    }
    Text_Score {
     id=#5
     possessed_class=CommonTextBlock
     tracks [{
       type=MovieSceneColorTrack
       display_name=Color and Opacity
       sections [{
         type=MovieSceneColorSection
         end_frame=6
         start_frame=0
         channels [
          Color.R {type=MovieSceneScriptingFloatChannel, const [0,2]}
          Color.G {type=MovieSceneScriptingFloatChannel, const [0,2]}
          Color.B {type=MovieSceneScriptingFloatChannel, const [0,2]}
          Color.A {type=MovieSceneScriptingFloatChannel, const [1,2]}
         ]
        }]
      }]
    }
    SongTitle {
     id=#6
     possessed_class=CommonTextBlock
     tracks [{
       type=MovieSceneColorTrack
       display_name=Color and Opacity
       sections [{
         type=MovieSceneColorSection
         end_frame=6
         start_frame=0
         channels [
          Color.R {type=MovieSceneScriptingFloatChannel, keys [[0,1],[6,0]]}
          Color.G {type=MovieSceneScriptingFloatChannel, keys [[0,1],[6,0]]}
          Color.B {type=MovieSceneScriptingFloatChannel, keys [[0,1],[6,0]]}
          Color.A {type=MovieSceneScriptingFloatChannel}
         ]
        }]
      }]
    }
    IntensitiesBorder {
     id=#7
     possessed_class=CommonBorder
     tracks [{
       type=MovieSceneColorTrack
       display_name=Content Color and Opacity
       sections [{
         type=MovieSceneColorSection
         end_frame=6
         start_frame=0
         channels [
          Color.R {type=MovieSceneScriptingFloatChannel}
          Color.G {type=MovieSceneScriptingFloatChannel}
          Color.B {type=MovieSceneScriptingFloatChannel}
          Color.A {type=MovieSceneScriptingFloatChannel, keys [[0,0],[6,1]]}
         ]
        }]
      }]
    }
    AlbumArtSizeBox {
     id=#8
     possessed_class=SizeBox
     tracks [
      {
       type=MovieSceneFloatTrack
       display_name=Width Override
       sections [{type=MovieSceneFloatSection, end_frame=6, start_frame=0, channels [None {type=MovieSceneScriptingFloatChannel, keys [[0,115],[6,250]]}]}]
      }
      {
       type=MovieSceneFloatTrack
       display_name=Height Override
       sections [{type=MovieSceneFloatSection, end_frame=6, start_frame=0, channels [None {type=MovieSceneScriptingFloatChannel, keys [[0,115],[6,250]]}]}]
      }
     ]
    }
    ScoreDetailsScaleBox {
     id=#9
     possessed_class=ScaleBox
     tracks [
      {
       type=MovieSceneFloatTrack
       display_name=User Specified Scale
       sections [{type=MovieSceneFloatSection, end_frame=6, start_frame=0, channels [None {type=MovieSceneScriptingFloatChannel, keys [[0,0],[6,1]]}]}]
      }
      {
       type=MovieSceneFloatTrack
       display_name=Render Opacity
       sections [{type=MovieSceneFloatSection, end_frame=6, start_frame=0, channels [None {type=MovieSceneScriptingFloatChannel, keys [[0,0],[6,1]]}]}]
      }
     ]
    }
    Text_Accuracy {
     id=#10
     possessed_class=CommonTextBlock
     tracks [{
       type=MovieSceneColorTrack
       display_name=Color and Opacity
       sections [{
         type=MovieSceneColorSection
         end_frame=6
         start_frame=0
         channels [
          Color.R {type=MovieSceneScriptingFloatChannel, keys [[0,1],[6,0]]}
          Color.G {type=MovieSceneScriptingFloatChannel, keys [[0,1],[6,0]]}
          Color.B {type=MovieSceneScriptingFloatChannel, keys [[0,1],[6,0]]}
          Color.A {type=MovieSceneScriptingFloatChannel, const [1,2]}
         ]
        }]
      }]
    }
   ]}
 }
 Unselected {
  binding_count=9
  length_seconds=0.3
  object=&self:Unselected
  sequence {bindings [
    QuestMarker {
     id=#2
     possessed_class=Image
     tracks [{
       type=MovieSceneColorTrack
       display_name=Color and Opacity
       sections [{
         type=MovieSceneColorSection
         end_frame=6
         start_frame=0
         channels [
          Color.R {type=MovieSceneScriptingFloatChannel, const [1,2]}
          Color.G {type=MovieSceneScriptingFloatChannel, const [1,2]}
          Color.B {type=MovieSceneScriptingFloatChannel, const [1,2]}
          Color.A {type=MovieSceneScriptingFloatChannel, const [1,2]}
         ]
        }]
      }]
    }
    SongArtist {
     id=#3
     possessed_class=CommonTextBlock
     tracks [{
       type=MovieSceneColorTrack
       display_name=Color and Opacity
       sections [{
         type=MovieSceneColorSection
         end_frame=6
         start_frame=0
         channels [
          Color.R {type=MovieSceneScriptingFloatChannel, const [1,2]}
          Color.G {type=MovieSceneScriptingFloatChannel, const [1,2]}
          Color.B {type=MovieSceneScriptingFloatChannel, const [1,2]}
          Color.A {type=MovieSceneScriptingFloatChannel, const [1,2]}
         ]
        }]
      }]
    }
    IntensitiesScaleBox {
     id=#4
     possessed_class=ScaleBox
     tracks [{
       type=MovieSceneFloatTrack
       display_name=User Specified Scale
       sections [{type=MovieSceneFloatSection, end_frame=6, start_frame=0, channels [None {type=MovieSceneScriptingFloatChannel, const [0,2]}]}]
      }]
    }
    Text_Score {
     id=#5
     possessed_class=CommonTextBlock
     tracks [{
       type=MovieSceneColorTrack
       display_name=Color and Opacity
       sections [{
         type=MovieSceneColorSection
         end_frame=6
         start_frame=0
         channels [
          Color.R {type=MovieSceneScriptingFloatChannel, const [0,2]}
          Color.G {type=MovieSceneScriptingFloatChannel, const [0,2]}
          Color.B {type=MovieSceneScriptingFloatChannel, const [0,2]}
          Color.A {type=MovieSceneScriptingFloatChannel, const [1,2]}
         ]
        }]
      }]
    }
    SongTitle {
     id=#6
     possessed_class=CommonTextBlock
     tracks [{
       type=MovieSceneColorTrack
       display_name=Color and Opacity
       sections [{
         type=MovieSceneColorSection
         end_frame=6
         start_frame=0
         channels [
          Color.R {type=MovieSceneScriptingFloatChannel, const [1,2]}
          Color.G {type=MovieSceneScriptingFloatChannel, const [1,2]}
          Color.B {type=MovieSceneScriptingFloatChannel, const [1,2]}
          Color.A {type=MovieSceneScriptingFloatChannel}
         ]
        }]
      }]
    }
    IntensitiesBorder {
     id=#7
     possessed_class=CommonBorder
     tracks [{
       type=MovieSceneColorTrack
       display_name=Content Color and Opacity
       sections [{
         type=MovieSceneColorSection
         end_frame=6
         start_frame=0
         channels [
          Color.R {type=MovieSceneScriptingFloatChannel}
          Color.G {type=MovieSceneScriptingFloatChannel}
          Color.B {type=MovieSceneScriptingFloatChannel}
          Color.A {type=MovieSceneScriptingFloatChannel, const [0,2]}
         ]
        }]
      }]
    }
    AlbumArtSizeBox {
     id=#8
     possessed_class=SizeBox
     tracks [
      {
       type=MovieSceneFloatTrack
       display_name=Width Override
       sections [{type=MovieSceneFloatSection, end_frame=6, start_frame=0, channels [None {type=MovieSceneScriptingFloatChannel, const [115,2]}]}]
      }
      {
       type=MovieSceneFloatTrack
       display_name=Height Override
       sections [{type=MovieSceneFloatSection, end_frame=6, start_frame=0, channels [None {type=MovieSceneScriptingFloatChannel, const [115,2]}]}]
      }
     ]
    }
    ScoreDetailsScaleBox {
     id=#9
     possessed_class=ScaleBox
     tracks [
      {
       type=MovieSceneFloatTrack
       display_name=User Specified Scale
       sections [{type=MovieSceneFloatSection, end_frame=6, start_frame=0, channels [None {type=MovieSceneScriptingFloatChannel, const [0,2]}]}]
      }
      {
       type=MovieSceneFloatTrack
       display_name=Render Opacity
       sections [{type=MovieSceneFloatSection, end_frame=6, start_frame=0, channels [None {type=MovieSceneScriptingFloatChannel, const [0,2]}]}]
      }
     ]
    }
    Text_Accuracy {
     id=#10
     possessed_class=CommonTextBlock
     tracks [{
       type=MovieSceneColorTrack
       display_name=Color and Opacity
       sections [{
         type=MovieSceneColorSection
         end_frame=6
         start_frame=0
         channels [
          Color.R {type=MovieSceneScriptingFloatChannel, const [1,2]}
          Color.G {type=MovieSceneScriptingFloatChannel, const [1,2]}
          Color.B {type=MovieSceneScriptingFloatChannel, const [1,2]}
          Color.A {type=MovieSceneScriptingFloatChannel, const [1,2]}
         ]
        }]
      }]
    }
   ]}
 }
]
graphs [
 EventGraph {nodes [
   PlaySound2D(Sound=/PilgrimQuickplay/Audio/FM_Player_Voted)
   > n1 = Switch on ESongDisplayStatus!(execute, Selection:ESongDisplayStatus Enum=Get DisplayStatus() -> (DisplayStatus:ESongDisplayStatus Enum=NewEnumerator2)) -> (NewEnumerator2:Exec, NewEnumerator0:Exec, NewEnumerator1:Exec) {is=title K2Node_SwitchEnum}
   ShowAlreadySelectedMessage(execute=@n1.NewEnumerator1)
   n2 = RefreshVisualsOnItemObjectUpdate!() -> (OutputDelegate, then, CantAddSongReason:EPilgrimCantAddSongReason Enum=None) {is=title K2Node_CustomEvent}
   > n3 = Sequence()
   n4 = Get ListItemObject() -> (ListItemObject:BP Song Entry Horizontal Item Obj Object Reference)
   > Set Song Display Status!(
    execute
    self
    SongStatus_SongShortName_19_#11:Name=GetShortName!(self:Sparks Song Data Object Reference=Get SongMetadata!(self:BP Song Entry Horizontal Item Obj Object Reference=@n4) -> (SongMetadata:Pilgrim Song Metadata Object Reference)) -> (ReturnValue:Name=None)
    SongStatus_DisplayStatus_14_#12:ESongDisplayStatus Enum=Get DisplayStatus!(self:BP Song Entry Horizontal Item Obj Object Reference=@n4) -> (DisplayStatus=NewEnumerator2)
    SongStatus_SongChooserPlayerState_17_#13:Player State Object Reference=Get SongAddedByPlayerState!(self:BP Song Entry Horizontal Item Obj Object Reference=@n4) -> (SongAddedByPlayerState:Player State Object Reference)
   ) -> (then) {is=title K2Node_CallFunction}
   n5 = Switch on EPilgrimCantAddSongReason!(execute=@n3.then_1, Selection:EPilgrimCantAddSongReason Enum=@n2.CantAddSongReason) -> (
    None:Exec
    SetlistLocked:Exec
    SongAlreadyAdded:Exec
    PlayerHasAddedMaxSongs:Exec
    InvalidSongShortName:Exec
    MissingSong:Exec
    SongNotAvailable:Exec
    FlaggedSongBlocked:Exec
   ) {is=title K2Node_SwitchEnum}
   ShowMaxSongsMessage(execute=@n5.PlayerHasAddedMaxSongs)
   ShowAlreadySelectedMessage(execute=@n5.SongAlreadyAdded)
   n6 = SetHighScore!() -> (OutputDelegate, then, PersonalBestResult:Pilgrim Personal Best Result Structure) {is=title K2Node_CustomEvent}
   n7 = Break Pilgrim Leaderboard Score Data!(PilgrimLeaderboardScoreData:"Pilgrim Leaderboard Score Data Structure (by ref)"=Break Pilgrim Leaderboard Entry!(PilgrimLeaderboardEntry:"Pilgrim Leaderboard Entry Structure (by ref)"=Break Pilgrim Personal Best Result!(PilgrimPersonalBestResult:"Pilgrim Personal Best Result Structure (by ref)"=@n6.PersonalBestResult) -> (
      BestSolo:Pilgrim Leaderboard Entry Structure
      BestBand:Pilgrim Leaderboard Entry Structure
      BestSpotlight:Pilgrim Leaderboard Entry Structure
      BestAlltime:Pilgrim Leaderboard Entry Structure
     ) {is=title K2Node_BreakStruct}) -> (
     bHasLocalPlayer:bool=~
     PlayerAccountIdToDisplayNameMap:Map of Strings to Strings
     Rank:int
     Percentile:float=~
     ScoreData:Pilgrim Leaderboard Score Data Structure
     DateTime:Date Time Structure
     FriendRank:int
    ) {is=title K2Node_BreakStruct}.ScoreData) -> (
    Score:int
    StarsEarned:int
    Accuracy:float=~
    bFullCombo:bool=~
    Difficulty:EPilgrimSongDifficulty Enum
    InstrumentsPlayed:Array of EPilgrimTrackType Enums
    SeasonNumber:int
   ) {is=title K2Node_BreakStruct}
   > Branch(Condition="integer > integer"(A=Length(TargetArray:Array of EPilgrimTrackType Enums=@n7.InstrumentsPlayed)))
   > SetVisibility(self=Get HighScoreInfo() -> (HighScoreInfo:Vertical Box Object Reference), InVisibility=SelfHitTestInvisible)
   > n8 = Sequence!(execute) -> (then_0, then_1, then_2:Exec, then_3:Exec, then_4:Exec)
   > "SetText (Text)"(self=Get Text_Score() -> (Text_Score:Common Text Object Reference), InText="To Text (Integer)"(Value=@n7))
   SetVisibility(
    execute=@n8.then_2
    self=Get FullComboNotif() -> (FullComboNotif:Border Object Reference)
    InVisibility=Select(Option 0=Collapsed, Option 1=SelfHitTestInvisible, Index=@n7.bFullCombo)
   )
   "SetText (Text)"(
    execute=@n8.then_3
    self=Get CommonTextBlock-Difficulty!(self:WBP Pilgrim Results Difficulty Object Reference=Get Difficulty() -> (Difficulty:WBP Pilgrim Results Difficulty Object Reference)) -> (CommonTextBlock-Difficulty:Common Text Object Reference)
    InText=Select!(
     DifficultyEasy:"Text (by ref)"=MakeLiteralText(Value="LOCTEXT(\"EASY\")")
     DifficultyMedium:"Text (by ref)"=MakeLiteralText(Value="LOCTEXT(\"MEDIUM\")")
     DifficultyHard:"Text (by ref)"=MakeLiteralText(Value="LOCTEXT(\"HARD\")")
     DifficultyExpert:"Text (by ref)"=MakeLiteralText(Value="LOCTEXT(\"EXPERT\")")
     None:"Text (by ref)"=MakeLiteralText(Value="LOCTEXT(\"NONE\")")
     Index:EPilgrimSongDifficulty Enum=@n7.Difficulty
    ) -> (ReturnValue:"Text (by ref)")
   )
   "SetText (Text)"(
    execute=@n8.then_1
    self=Get Text_Accuracy() -> (Text_Accuracy:Common Text Object Reference)
    InText=Format Text!(Format="LOCTEXT(\"{Accuracy}%\")", Accuracy:float=@n7.Accuracy) -> (Result)
   )
   SetupStarsResults_DEPRECATED!(
    execute=@n8.then_4
    self:WBP Pilgrim Quickplay Local Star Bar Object Reference=Get WBP_PilgrimQuickplay_LocalStarBar() -> (WBP_PilgrimQuickplay_LocalStarBar:WBP Pilgrim Quickplay Local Star Bar Object Reference)
    Instrument:EPilgrimTrackType Enum="Get (a copy)"(Array:Array of EPilgrimTrackType Enums=@n7.InstrumentsPlayed) -> (Output:EPilgrimTrackType Enum)
    StarCount:int=@n7.StarsEarned
   ) -> (then)
   n9 = SetAddedByPlayerInfo!() -> (OutputDelegate, then, FortPlayerStateAthena:Fort Player State Athena Object Reference) {is=title K2Node_CustomEvent}
   > n10 = Sequence!(execute) -> (then_0, then_1)
   > "SetText (Text)"(
    self=Get AddedByPlayerText() -> (AddedByPlayerText:Common Text Object Reference)
    InText=Format Text!(
     Format=Get AddedByPlayer() -> (AddedByPlayer:Text)
     0:String=GetPlayerName!(self:Player State Object Reference=@n9.FortPlayerStateAthena) -> (ReturnValue:String)
    ) -> (Result)
   )
   n11 = GetDynamicMaterial(execute=@n10.then_1, self=Get AddedByAvatar() -> (AddedByAvatar:Image Object Reference))
   > n12 = Sequence()
   > SetVectorParameterValue(
    self=@n11
    ParameterName=BG_Color
    Value=GetPinColor!(self:Fort Player State Zone Object Reference=@n9.FortPlayerStateAthena) -> (ReturnValue:Linear Color Structure)
   )
   n13 = Async Load Asset(
    execute=@n12.then_1
    Asset=GetSocialAvatarBrush!(self:Fort Player State Zone Object Reference=@n9.FortPlayerStateAthena, bSmallImage:bool=true) -> (ReturnValue:Texture 2D Soft Object Reference)
   )
   n14 = Cast To Texture2D(execute=@n13.Completed, Object=@n13)
   > SetTextureParameterValue(self=@n11, ParameterName=PlayerIcon, Value=@n14)
   Event Construct()
   > SetRenderOpacity(self=Get MessageOverlay() -> (MessageOverlay:Overlay Object Reference), InOpacity=0)
   Event On Selected()
   > PlayAnimation(
    InAnimation=Get Selected_Animated() -> (Selected_Animated:Widget Animation Object Reference)
    PlaybackSpeed=Get SelectAnimPlaybackSpeed() -> (SelectAnimPlaybackSpeed:double)
   )
   n15 = Event On Deselected()
   PlayAnimation(execute=[@n15.then, @n24.then], InAnimation=Get Selected_Animated(), PlayMode=Reverse, PlaybackSpeed=Get SelectAnimPlaybackSpeed())
   n16 = Event OnListItemObjectSet()
   > n17 = Sequence!(execute) -> (then_0, then_1, then_2:Exec)
   > n18 = Cast To BP_SongEntryHorizontal_ItemObj!(execute, Object=@n16.ListItemObject) -> (then, CastFailed, AsBP Song Entry Horizontal Item Obj:BP Song Entry Horizontal Item Obj Object Reference) {is=title K2Node_DynamicCast}
   > n19 = Set ListItemObject(ListItemObject:BP Song Entry Horizontal Item Obj Object Reference=@n18) -> (Output_Get:BP Song Entry Horizontal Item Obj Object Reference)
   n20 = Get SongMetadata(self=@n19)
   > Is Valid(InputObject=@n20)
   > SetSongMetadata!(execute, self, NewSongMetadata:Pilgrim Song Metadata Object Reference=@n20) -> (then)
   > Set Song Display Status(
    SongStatus_SongShortName_19_#11=GetShortName(self=Get SongMetadata(self=@n18)) -> (ReturnValue=None)
    SongStatus_DisplayStatus_14_#12=Get DisplayStatus(self=@n18) -> (DisplayStatus=NewEnumerator2)
    SongStatus_SongChooserPlayerState_17_#13=Get SongAddedByPlayerState(self=@n18)
   )
   > Bind Event to On List Item Object Updated(self:BP Song Entry Horizontal Item Obj Object Reference=@n18, Delegate=@n2) {is=title K2Node_AddDelegate}
   n21 = MakeGameplayTagContainerFromArray!(GameplayTags:Array of Gameplay Tag Structures=Make Array!() -> (Array)) -> (ReturnValue:Gameplay Tag Container Structure)
   n22 = Make Array!("[0]":Name=GetShortName(self=Get SongMetadata!() -> (SongMetadata)) -> (ReturnValue=None)) -> (Array:Array of Names)
   n23 = DoAnyQuestsPassFilter!(
    execute=@n17.then_1
    Player:Player State Object Reference=Get PlayerState!(self:Controller Object Reference=GetOwningPlayer()) -> (PlayerState:Player State Object Reference)
    Filter:"Pilgrim Quest Filter Structure (by ref)"=Make Pilgrim Quest Filter!(
     QuestTagQuery:Gameplay Tag Query Structure=(TokenStreamVersion=0,TagDictionary=,QueryTokenStream=,UserDescription="",AutoDescription="")
     QuestObjectiveFilterQuery:Gameplay Tag Query Structure=MakeGameplayTagQuery_MatchAnyTags!(InTags:"Gameplay Tag Container Structure (by ref)"=@n21) -> (ReturnValue:Gameplay Tag Query Structure)
     QuestObjectiveFilterMatcher:Song Short Name Matcher Structure=Make Song Short Name Matcher!(SongShortNames:Array of Names=@n22, MatchType:ESongShortNameMatchType Enum=MatchAny) -> (SongShortNameMatcher:Song Short Name Matcher Structure) {is=title K2Node_MakeStruct}
     QuestObjectiveSongSIDMatcher:Song Item Def Primary Asset Name Matcher Structure=(SongItemDefs=,MatchType=AlwaysMatch)
     bShowCompletedQuests:bool=True
     bShowUpdatedQuests:bool=True
     bShowUnchangedQuests:bool=True
    ) -> (PilgrimQuestFilter:Pilgrim Quest Filter Structure) {is=title K2Node_MakeStruct}
   ) -> (then, ReturnValue:bool)
   > SetVisibility(
    self=Get QuestMarker() -> (QuestMarker:Image Object Reference)
    InVisibility=Select!(Option 0:ESlateVisibility Enum=Hidden, Option 1:ESlateVisibility Enum=SelfHitTestInvisible, Index:bool=@n23) -> (ReturnValue:ESlateVisibility Enum)
   )
   SetVisibility(execute=@n17.then_2, self=Get HighScoreInfo(), InVisibility=Hidden)
   n24 = Event On Entry Released() {is=title K2Node_Event}
   Event OnInitialized()
   > Make SongTag Map()
   PlayAnimation(InAnimation=Get Unselected() -> (Unselected:Widget Animation Object Reference))
   PlayAnimation(InAnimation=Get Selected() -> (Selected:Widget Animation Object Reference))
   Event PreConstruct()
   Event Tick()
   Event On Clicked()
  ]}
 GetSongMetadata {nodes [
   GetSongMetadata() {is=title K2Node_FunctionEntry}
   > Return Node!(execute, SongMetadata:Pilgrim Song Metadata Object Reference=Get SongMetadata())
  ]}
 Make SongTag Map {nodes [
   Make SongTag Map() {is=title K2Node_FunctionEntry}
   > Add!(
    execute
    TargetMap:Map of Gameplay Tag Structures to Image Object References=Get SongTagToInstruments() -> (SongTagToInstruments:Map of Gameplay Tag Structures to Image Object References)
    Key:"Gameplay Tag Structure (by ref)"=(TagName="Item.Instrument.TrackType.Bass")
    Value:"Image Object Reference (by ref)"=Get SpotlightMarker_Bass() -> (SpotlightMarker_Bass:Image Object Reference)
   ) -> (then) {is=title K2Node_CallFunction}
   > Add(
    TargetMap=Get SongTagToInstruments()
    Key=(TagName="Item.Instrument.TrackType.Drums")
    Value=Get SpotlightMarker_Drums() -> (SpotlightMarker_Drums:Image Object Reference)
   )
   > Add(
    TargetMap=Get SongTagToInstruments()
    Key=(TagName="Item.Instrument.TrackType.Guitar")
    Value=Get SpotlightMarker_Guitar() -> (SpotlightMarker_Guitar:Image Object Reference)
   )
   > Add(
    TargetMap=Get SongTagToInstruments()
    Key=(TagName="Item.Instrument.TrackType.Vocals")
    Value=Get SpotlightMarker_Vocals() -> (SpotlightMarker_Vocals:Image Object Reference)
   )
  ]}
 PlayMessageAnimation {nodes [
   n1 = PlayMessageAnimation!() -> (then, MessageText:Text) {is=title K2Node_FunctionEntry}
   > "SetText (Text)"(self=Get MessageText() -> (MessageText:Common Text Object Reference), InText=@n1)
   > PlayAnimation(InAnimation=Get FadeMessage() -> (FadeMessage:Widget Animation Object Reference))
  ]}
 Set Song Display Status {nodes [
   n1 = Set Song Display Status!() -> (then, SongStatus:Song Display Status Struct Structure) {is=title K2Node_FunctionEntry}
   n2 = Break Song Display Status Struct!(SongDisplayStatusStruct:"Song Display Status Struct Structure (by ref)"=@n1) -> (
    SongShortName_19_#11:Name
    DisplayStatus_14_#12:ESongDisplayStatus Enum
    SongChooserPlayerState_17_#13:Player State Object Reference
   ) {is=title K2Node_BreakStruct}
   > n3 = Set DisplayStatus(DisplayStatus:ESongDisplayStatus Enum=@n2.DisplayStatus_14_#12) -> (Output_Get:ESongDisplayStatus Enum=NewEnumerator2)
   > SetVisibility(
    self=Get AddedToSetlistIndicator() -> (AddedToSetlistIndicator:Overlay Object Reference)
    InVisibility=Select!(
     NewEnumerator2:ESlateVisibility Enum=Hidden
     NewEnumerator0:ESlateVisibility Enum=SelfHitTestInvisible
     NewEnumerator1:ESlateVisibility Enum=Hidden
     Index:ESongDisplayStatus Enum=@n3
    ) -> (ReturnValue)
   )
   > SetVisibility(
    self=Get AlreadyAddedOverlay() -> (AlreadyAddedOverlay:Overlay Object Reference)
    InVisibility=Select(NewEnumerator2=Hidden, NewEnumerator0=Hidden, NewEnumerator1=SelfHitTestInvisible, Index=@n3)
   )
   > Is Valid(InputObject=@n2.SongChooserPlayerState_17_#13)
   > n4 = Cast To FortPlayerStateAthena!(execute, Object=@n2.SongChooserPlayerState_17_#13) -> (then, CastFailed, AsFort Player State Athena:Fort Player State Athena Object Reference) {is=title K2Node_DynamicCast}
   > SetAddedByPlayerInfo!(execute, self, FortPlayerStateAthena:Fort Player State Athena Object Reference=@n4) -> (then) {is=title K2Node_CallFunction}
  ]}
 SetAllIntensities {nodes [
   n1 = SetAllIntensities!() -> (then, SongMetadata:Pilgrim Song Metadata Object Reference) {is=title K2Node_FunctionEntry}
   > n2 = Sequence!(execute) -> (then_0, then_1, then_2, then_3:Exec)
   > SetIntensity!(
    execute
    self
    IntensityWidget:WBP Pilgrim Quickplay Song Intensity Entry Object Reference=Get GuitarIntensity() -> (GuitarIntensity:WBP Pilgrim Quickplay Song Intensity Entry Object Reference)
    SongMetadata:Pilgrim Song Metadata Object Reference=@n1
    Instrument:EPilgrimTrackType Enum=TrackGuitar
    InstrumentIcon:ESparksInstrumentType Enum=GetStartingInstrumentType!(self:Pilgrim Song Metadata Object Reference=@n1, TrackType:EPilgrimTrackType Enum=TrackGuitar) -> (ReturnValue:ESparksInstrumentType Enum=Guitar)
   ) -> (then)
   SetIntensity(
    execute=@n2.then_1
    IntensityWidget=Get DrumsIntensity() -> (DrumsIntensity:WBP Pilgrim Quickplay Song Intensity Entry Object Reference)
    SongMetadata=@n1
    Instrument=TrackDrum
    InstrumentIcon=GetStartingInstrumentType(self=@n1, TrackType=TrackDrum) -> (ReturnValue=Guitar)
   )
   SetIntensity(
    execute=@n2.then_2
    IntensityWidget=Get VocalIntensity() -> (VocalIntensity:WBP Pilgrim Quickplay Song Intensity Entry Object Reference)
    SongMetadata=@n1
    Instrument=TrackVocals
    InstrumentIcon=GetStartingInstrumentType(self=@n1, TrackType=TrackVocals) -> (ReturnValue=Guitar)
   )
   SetIntensity(
    execute=@n2.then_3
    IntensityWidget=Get BassIntensity() -> (BassIntensity:WBP Pilgrim Quickplay Song Intensity Entry Object Reference)
    SongMetadata=@n1
    Instrument=TrackBass
    InstrumentIcon=GetStartingInstrumentType(self=@n1, TrackType=TrackBass) -> (ReturnValue=Guitar)
   )
  ]}
 SetIntensity {nodes [
   n1 = SetIntensity!() -> (
    then
    IntensityWidget:WBP Pilgrim Quickplay Song Intensity Entry Object Reference
    SongMetadata:Pilgrim Song Metadata Object Reference
    Instrument:EPilgrimTrackType Enum=TrackGuitar
    InstrumentIcon:ESparksInstrumentType Enum=Guitar
   ) {is=title K2Node_FunctionEntry}
   > Update Intensity for Instrument!(
    execute
    self:WBP Pilgrim Quickplay Song Intensity Entry Object Reference=@n1
    Intensity:int=GetTrackIntensity!(self:Pilgrim Song Metadata Object Reference=@n1.SongMetadata, TrackType:EPilgrimTrackType Enum=@n1.Instrument) -> (ReturnValue:int=0)
    TrackType:EPilgrimTrackType Enum=@n1.Instrument
    StartingInstrument:ESparksInstrumentType Enum=@n1.InstrumentIcon
    Is3D_deprecated:bool=false
    SongMetadata:Pilgrim Song Metadata Object Reference
   ) -> (then)
  ]}
 SetSongMetadata {nodes [
   n1 = SetSongMetadata!() -> (then, NewSongMetadata:Pilgrim Song Metadata Object Reference) {is=title K2Node_FunctionEntry}
   > n2 = Set SongMetadata(SongMetadata:Pilgrim Song Metadata Object Reference=@n1) -> (Output_Get:Pilgrim Song Metadata Object Reference)
   > n3 = Sequence!(execute) -> (then_0, then_1, then_2, then_3, then_4:Exec)
   > "SetText (Text)"(
    self=Get SongTitle() -> (SongTitle:Common Text Object Reference)
    InText=GetTitleText!(self:Sparks Song Data Object Reference=@n2) -> (ReturnValue:Text)
   )
   SetAllIntensities!(execute=@n3.then_3, self, SongMetadata:Pilgrim Song Metadata Object Reference=@n2) -> (then) {is=title K2Node_CallFunction}
   SetSpotlightStatus!(execute=@n3.then_4, self, Metadata:Pilgrim Song Metadata Object Reference=@n2) -> (then)
   "SetText (Text)"(
    execute=@n3.then_1
    self=Get SongArtist() -> (SongArtist:Common Text Object Reference)
    InText=GetArtistText!(self:Sparks Song Data Object Reference=@n2) -> (ReturnValue:Text)
   )
   LazyLoadAlbumArt!(
    execute=@n3.then_2
    Song Metadata:Pilgrim Song Metadata Object Reference=@n2
    Lazy Image:Fort Lazy Image Object Reference=Get AlbumArt() -> (AlbumArt:Fort Lazy Image Object Reference)
   ) -> (then)
  ]}
 SetSpotlightStatus {nodes [
   n1 = SetSpotlightStatus!() -> (then, Metadata:Pilgrim Song Metadata Object Reference) {is=title K2Node_FunctionEntry}
   > n2 = Values!(execute, TargetMap:Map of Gameplay Tag Structures to Image Object References=Get SongTagToInstruments()) -> (then, Values:Array of Image Object References)
   > n3 = For Each Loop(Array:Array of Image Object References=@n2) -> (Array Element:Image Object Reference)
   n4 = GetOwningPlayer()
   n5 = GetShortName(self=@n1) -> (ReturnValue=None)
   n6 = For Each Loop(
    Exec=@n3.Completed
    Array:Array of Pilgrim Spotlight Competition Structures=GetSpotlightCompetitions!(
     self:Pilgrim Leaderboard Player Component Object Reference=GetComponentByClass!(self:Actor Object Reference=@n4, ComponentClass:Actor Component Class Reference=/Script/PilgrimCoreRuntime.PilgrimLeaderboardPlayerComponent) -> (ReturnValue:Pilgrim Leaderboard Player Component Object Reference)
     SpotlightParams:"Pilgrim Get Spotlight Params Structure (by ref)"=Make Pilgrim Get Spotlight Params!(
      SongShortName:Name=@n5
      InstrumentType:EPilgrimTrackType Enum=TrackNone
      StartsBefore:Date Time Structure=0001.01.01-00.00.00
      EndsAfter:Date Time Structure=0001.01.01-00.00.00
     ) -> (PilgrimGetSpotlightParams:Pilgrim Get Spotlight Params Structure) {is=title K2Node_MakeStruct}
    ) -> (OutCompetitions:Array of Pilgrim Spotlight Competition Structures)
   ) -> (Array Element:Pilgrim Spotlight Competition Structure)
   > n7 = For Each Loop(Array:Array of Gameplay Tag Structures=BreakGameplayTagContainer!(GameplayTagContainer:"Gameplay Tag Container Structure (by ref)"=Break Pilgrim Spotlight Competition!(PilgrimSpotlightCompetition:"Pilgrim Spotlight Competition Structure (by ref)"=@n6) -> (
      SpotlightName:String
      SpotlightGroup:String
      BeginTime:Date Time Structure
      EndTime:Date Time Structure
      SongUIDs:Array of Strings
      SongShortNames:Array of Names
      Constraints:Gameplay Tag Container Structure
     ) {is=title K2Node_BreakStruct}.Constraints) -> (GameplayTags:Array of Gameplay Tag Structures)) -> (Array Element:Gameplay Tag Structure)
   n8 = Find(TargetMap:Map of Gameplay Tag Structures to Image Object References=Get SongTagToInstruments(), Key:"Gameplay Tag Structure (by ref)"=@n7) -> (Value:Image Object Reference)
   > Branch(Condition=@n8.ReturnValue)
   > Set Visibility(Visibility=SelfHitTestInvisible, self=@n8)
   Set Visibility(execute=@n3.LoopBody, Visibility=Hidden, self=@n3)
  ]}
 ShowAlreadySelectedMessage {nodes [
   ShowAlreadySelectedMessage() {is=title K2Node_FunctionEntry}
   > PlayMessageAnimation!(execute, self, MessageText:Text=Get AlreadySelected() -> (AlreadySelected:Text)) -> (then) {is=title K2Node_CallFunction}
  ]}
 ShowMaxSongsMessage {nodes [
   ShowMaxSongsMessage() {is=title K2Node_FunctionEntry}
   > PlayMessageAnimation(MessageText=Get AtMaxSongs() -> (AtMaxSongs:Text))
  ]}
]
variables [
 AddedByPlayer {type=string, default="Added by {0}"}
 AlreadySelected {type=string, default=Already Added}
 AtMaxSongs {type=string, default=max songs chosen}
 DisplayStatus {type=ESongDisplayStatus, default=None}
 Hovered Color {type=LinearColor, default=<rgba 0 0 0 0>}
 HoveredColor {type=LinearColor, default=<rgba 0 0.3125 0.625 1>}
 ListItemObject {type=BP_SongEntryHorizontal_ItemObj_C, default=null}
 NormalColor {type=LinearColor, default=<rgba 0 0.0356 0.2122 1>}
 SelectAnimPlaybackSpeed {type=number, default=5}
 SelectedColor {type=LinearColor, default=<rgba 1 0.1231 0 1>}
 SelectedHoveredColor {type=LinearColor, default=<rgba 1 0.301 0 1>}
 SongMetadata {type=PilgrimSongMetadata, default=null}
 SongTagToInstruments {default {}}
]
widgets {
 info {inheritedWidgetCount=1, parentClass=&CommonButtonBase, rootWidgetClass=&Overlay, widgetCount=52}
 widgets [
  Overlay_0 {widgetClassPath=&Overlay, widget=Overlay_0}
  VerticalBox_21 {
   widgetClassPath=&VerticalBox
   parent=Overlay_0
   widget=VerticalBox_21
   slot {class=OverlaySlot, properties {horizontalAlignment=HAlign_Fill, verticalAlignment=VAlign_Center}}
  }
  HorizontalBox_34 {
   widgetClassPath=&HorizontalBox
   parent=VerticalBox_21
   widget=HorizontalBox_34
   slot {class=VerticalBoxSlot, properties {verticalAlignment=VAlign_Center}}
  }
  AlbumArtSizeBox {
   widgetClassPath=&SizeBox
   parent=HorizontalBox_34
   widget=AlbumArtSizeBox
   properties {
    bOverride_HeightOverride=true
    bOverride_MaxAspectRatio=true
    bOverride_MinAspectRatio=true
    bOverride_WidthOverride=true
    heightOverride=115
    widthOverride=115
   }
   slot {class=HorizontalBoxSlot, properties {verticalAlignment=VAlign_Top}}
  }
  AlbumArt {
   widgetClassPath=&FortLazyImage
   bIsVariable=true
   parent=AlbumArtSizeBox
   widget=AlbumArt
   properties {brush {
     bIsDynamicallyLoaded=false
     drawAs=Image
     imageType=NoImage
     mirroring=NoMirror
     resourceName=None
     resourceObject=/PilgrimCore/Widgets/Textures/T_HMX_Disc_01_D
     tiling=NoTile
     imageSize=<2048 2048>
     margin=<ltrb 0 0 0 0>
     outlineSettings {
      bUseBrushTransparency=false
      roundingType=HalfHeightRadius
      width=0
      color {colorUseRule=UseColor_Specified, specifiedColor=<rgba 0 0 0 0>}
      cornerRadii=<0 0 0 0>
     }
     tintColor {colorUseRule=UseColor_Specified, specifiedColor=<rgba 1 1 1 1>}
     uVRegion {bIsValid=false, max=<0 0>, min=<0 0>}
    }}
   slot {class=SizeBoxSlot}
  }
  SongInfoContent {
   widgetClassPath=&CommonBorder
   parent=HorizontalBox_34
   widget=SongInfoContent
   properties {verticalAlignment=VAlign_Center, padding=<ltrb 0 0 0 0>}
   slot {class=HorizontalBoxSlot, properties {verticalAlignment=VAlign_Center, padding=<ltrb 25 0 25 0>, size {sizeRule=Fill, value=1}}}
  }
  SongInfoBox {
   widgetClassPath=&HorizontalBox
   parent=SongInfoContent
   widget=SongInfoBox
   slot {class=BorderSlot, properties {verticalAlignment=VAlign_Center, padding=<ltrb 0 0 0 0>}}
  }
  TitleArtistBox {widgetClassPath=&VerticalBox, parent=SongInfoBox, widget=TitleArtistBox, slot {class=HorizontalBoxSlot}}
  SongTitle {
   widgetClassPath=&CommonTextBlock
   bIsVariable=true
   parent=TitleArtistBox
   widget=SongTitle
   properties {
    scrollStyle=/PilgrimQuickplay/Widgets/ScrollStyles/TextScrollStyle_PilgrimQuickplay_Setlist.TextScrollStyle_PilgrimQuickplay_Setlist_C
    style=/FortUILibrary/Styles/Typography/Foundation/TextStyle_BurbankSmall_M_White.TextStyle_BurbankSmall_M_White_C
    text=SONG TITLE
    textTransformPolicy=ToUpper
    visibility=HitTestInvisible
    font {
     bForceMonospaced=false
     bMaterialIsStencil=false
     fontMaterial=null
     fontObject=&BurbankSmall
     letterSpacing=0
     monospacedWidth=1
     size=24
     skewAmount=0
     typefaceFontName=Black
     outlineSettings {
      bApplyOutlineToDropShadows=false
      bMiteredCorners=false
      bSeparateFillAlpha=false
      outlineMaterial=null
      outlineSize=0
      outlineColor=<rgba 0 0 0 1>
     }
    }
    shadowOffset=<0 0>
   }
   slot {class=VerticalBoxSlot}
  }
  Spacer_123 {widgetClassPath=&Spacer, parent=TitleArtistBox, widget=Spacer_123, slot {class=VerticalBoxSlot}}
  SongArtist {
   widgetClassPath=&CommonTextBlock
   bIsVariable=true
   parent=TitleArtistBox
   widget=SongArtist
   properties {
    clipping=OnDemand
    scrollStyle=&TextScrollStyle_PilgrimQuickplay_Setlist_C
    style=/FortUILibrary/Styles/Typography/Foundation/TextStyle-BurbankSmall-XS-White.TextStyle-BurbankSmall-XS-White_C
    text=ARTIST
    textTransformPolicy=ToUpper
    visibility=HitTestInvisible
    font {
     bForceMonospaced=false
     bMaterialIsStencil=false
     fontMaterial=null
     fontObject=&BurbankSmall
     letterSpacing=0
     monospacedWidth=1
     size=13
     skewAmount=0
     typefaceFontName=Black
     outlineSettings {
      bApplyOutlineToDropShadows=false
      bMiteredCorners=false
      bSeparateFillAlpha=false
      outlineMaterial=null
      outlineSize=0
      outlineColor=<rgba 0 0 0 1>
     }
    }
    shadowOffset=<0 0>
   }
   slot {class=VerticalBoxSlot}
  }
  IntensitiesBorder {
   widgetClassPath=&CommonBorder
   parent=TitleArtistBox
   widget=IntensitiesBorder
   properties {contentColorAndOpacity=<rgba 0 0 0 0>, padding=<ltrb 0 0 0 0>}
   slot {class=VerticalBoxSlot, properties {size {sizeRule=Fill, value=1}}}
  }
  IntensitiesScaleBox {
   widgetClassPath=&ScaleBox
   parent=IntensitiesBorder
   widget=IntensitiesScaleBox
   properties {stretch=UserSpecified, userSpecifiedScale=0}
   slot {class=BorderSlot, properties {padding=<ltrb 0 0 0 0>}}
  }
  Intensities {
   widgetClassPath=/Script/UMG.UniformGridPanel
   parent=IntensitiesScaleBox
   widget=Intensities
   slot {class=ScaleBoxSlot, properties {horizontalAlignment=HAlign_Fill, verticalAlignment=VAlign_Fill}}
  }
  GuitarIntensity {
   widgetClassPath=&/WBP_PilgrimQuickplay_SongIntensityEntry
   bIsVariable=true
   parent=Intensities
   widget=GuitarIntensity
   properties {visibility=HitTestInvisible}
   slot {class=UniformGridSlot}
  }
  DrumsIntensity {
   widgetClassPath=&WBP_PilgrimQuickplay_SongIntensityEntry
   bIsVariable=true
   parent=Intensities
   widget=DrumsIntensity
   properties {visibility=HitTestInvisible}
   slot {class=UniformGridSlot, properties {column=1}}
  }
  VocalIntensity {
   widgetClassPath=&WBP_PilgrimQuickplay_SongIntensityEntry
   bIsVariable=true
   parent=Intensities
   widget=VocalIntensity
   properties {visibility=HitTestInvisible}
   slot {class=UniformGridSlot, properties {row=1}}
  }
  BassIntensity {
   widgetClassPath=&WBP_PilgrimQuickplay_SongIntensityEntry
   bIsVariable=true
   parent=Intensities
   widget=BassIntensity
   properties {visibility=HitTestInvisible}
   slot {class=UniformGridSlot, properties {column=1, row=1}}
  }
  Spacer_106 {
   widgetClassPath=&Spacer
   parent=SongInfoBox
   widget=Spacer_106
   properties {size=<10 1>}
   slot {class=HorizontalBoxSlot, properties {size {sizeRule=Fill, value=1}}}
  }
  SizeBox_289 {
   widgetClassPath=&SizeBox
   parent=SongInfoBox
   widget=SizeBox_289
   properties {bOverride_MinDesiredHeight=true, minDesiredHeight=36}
   slot {class=HorizontalBoxSlot, properties {verticalAlignment=VAlign_Bottom}}
  }
  HighScoreInfo {
   widgetClassPath=&VerticalBox
   bIsVariable=true
   parent=SizeBox_289
   widget=HighScoreInfo
   properties {visibility=Hidden}
   slot {class=SizeBoxSlot, properties {verticalAlignment=VAlign_Bottom}}
  }
  StarsAndAccuracy {
   widgetClassPath=&HorizontalBox
   parent=HighScoreInfo
   widget=StarsAndAccuracy
   slot {class=VerticalBoxSlot, properties {horizontalAlignment=HAlign_Right}}
  }
  WBP_PilgrimQuickplay_LocalStarBar {
   widgetClassPath=/PilgrimQuickplay/Widgets/WBP_PilgrimQuickplay_LocalStarBar
   bIsVariable=true
   parent=StarsAndAccuracy
   widget=WBP_PilgrimQuickplay_LocalStarBar
   properties {in User Specified Scale=0.45, scale Stretch=UserSpecified}
   slot {class=HorizontalBoxSlot}
  }
  FullComboNotif {
   widgetClassPath=&Border
   bIsVariable=true
   parent=StarsAndAccuracy
   widget=FullComboNotif
   properties {horizontalAlignment=HAlign_Center, verticalAlignment=VAlign_Center, visibility=Collapsed, brushColor=<rgba 1 1 1 0>, padding=<ltrb 5 0 5 0>}
   slot {class=HorizontalBoxSlot, properties {verticalAlignment=VAlign_Center, padding=<ltrb 5 0 5 0>, size {sizeRule=Automatic, value=0.03}}}
  }
  WBP_PilgrimQuickplay_FullComboIndicator {
   widgetClassPath=/PilgrimQuickplay/Widgets/WBP_PilgrimQuickplay_FullComboIndicator
   bIsVariable=true
   parent=FullComboNotif
   widget=WBP_PilgrimQuickplay_FullComboIndicator
   properties {showIconOnly=true}
   slot {class=BorderSlot, properties {horizontalAlignment=HAlign_Center, verticalAlignment=VAlign_Center, padding=<ltrb 5 0 5 0>}}
  }
  Text_Accuracy {
   widgetClassPath=&CommonTextBlock
   bIsVariable=true
   parent=StarsAndAccuracy
   widget=Text_Accuracy
   properties {
    scrollStyle=&TextScrollStyle-Base_C
    style=/FortUILibrary/Styles/Typography/Foundation/TextStyle_BurbankSmall_S_White.TextStyle_BurbankSmall_S_White_C
    text=100%
    font {
     bForceMonospaced=false
     bMaterialIsStencil=false
     fontMaterial=null
     fontObject=&BurbankSmall
     letterSpacing=0
     monospacedWidth=1
     size=18
     skewAmount=0
     typefaceFontName=Black
     outlineSettings {
      bApplyOutlineToDropShadows=false
      bMiteredCorners=false
      bSeparateFillAlpha=false
      outlineMaterial=null
      outlineSize=0
      outlineColor=<rgba 0 0 0 1>
     }
    }
    shadowOffset=<0 0>
   }
   slot {
    class=HorizontalBoxSlot
    properties {horizontalAlignment=HAlign_Left, verticalAlignment=VAlign_Center, padding=<ltrb 5 0 5 0>, size {sizeRule=Automatic, value=0.05}}
   }
  }
  ScoreDetailsScaleBox {
   widgetClassPath=&ScaleBox
   parent=HighScoreInfo
   widget=ScoreDetailsScaleBox
   properties {stretch=UserSpecified, userSpecifiedScale=0}
   slot {class=VerticalBoxSlot, properties {horizontalAlignment=HAlign_Right}}
  }
  ScoreDetails {
   widgetClassPath=&HorizontalBox
   parent=ScoreDetailsScaleBox
   widget=ScoreDetails
   slot {class=ScaleBoxSlot, properties {horizontalAlignment=HAlign_Fill, verticalAlignment=VAlign_Fill}}
  }
  InstrumentBorder {
   widgetClassPath=&CommonBorder
   parent=ScoreDetails
   widget=InstrumentBorder
   properties {contentColorAndOpacity=<rgba 0 0 0 1>, padding=<ltrb 0 0 0 0>}
   slot {class=HorizontalBoxSlot, properties {padding=<ltrb 5 0 5 0>}}
  }
  InstrumentImg {
   widgetClassPath=&Image
   bIsVariable=true
   parent=InstrumentBorder
   widget=InstrumentImg
   properties {brush {
     bIsDynamicallyLoaded=false
     drawAs=Image
     imageType=NoImage
     mirroring=NoMirror
     resourceName=None
     resourceObject=/PilgrimQuickplay/Widgets/Materials/M_UI_Icon_Instrument_Type_Sm
     tiling=NoTile
     imageSize=<48 48>
     margin=<ltrb 0 0 0 0>
     outlineSettings {
      bUseBrushTransparency=false
      roundingType=HalfHeightRadius
      width=0
      color {colorUseRule=UseColor_Specified, specifiedColor=<rgba 0 0 0 0>}
      cornerRadii=<0 0 0 0>
     }
     tintColor {colorUseRule=UseColor_Specified, specifiedColor=<rgba 1 1 1 1>}
     uVRegion {bIsValid=false, max=<0 0>, min=<0 0>}
    }}
   slot {class=BorderSlot, properties {padding=<ltrb 0 0 0 0>}}
  }
  Difficulty {
   widgetClassPath=/PilgrimQuickplay/Widgets/Results/WBP_Pilgrim_Results_Difficulty
   bIsVariable=true
   parent=ScoreDetails
   widget=Difficulty
   slot {class=HorizontalBoxSlot, properties {horizontalAlignment=HAlign_Center, verticalAlignment=VAlign_Center, padding=<ltrb 5 0 5 0>}}
  }
  Text_Score {
   widgetClassPath=&CommonTextBlock
   bIsVariable=true
   parent=ScoreDetails
   widget=Text_Score
   properties {
    justification=Center
    style=&TextStyle_BurbankSmall_S_White_C
    text="99,999,999"
    font {
     bForceMonospaced=false
     bMaterialIsStencil=false
     fontMaterial=null
     fontObject=&BurbankSmall
     letterSpacing=0
     monospacedWidth=1
     size=18
     skewAmount=0
     typefaceFontName=Black
     outlineSettings {
      bApplyOutlineToDropShadows=false
      bMiteredCorners=false
      bSeparateFillAlpha=false
      outlineMaterial=null
      outlineSize=0
      outlineColor=<rgba 0 0 0 1>
     }
    }
    shadowOffset=<0 0>
   }
   slot {class=HorizontalBoxSlot, properties {verticalAlignment=VAlign_Center, padding=<ltrb 5 0 5 0>}}
  }
  StatusIcons {widgetClassPath=&Overlay, parent=HorizontalBox_34, widget=StatusIcons, slot {class=HorizontalBoxSlot}}
  VerticalBox_0 {
   widgetClassPath=&VerticalBox
   parent=StatusIcons
   widget=VerticalBox_0
   slot {class=OverlaySlot, properties {verticalAlignment=VAlign_Bottom, padding=<ltrb 5 5 5 5>}}
  }
  VerticalBox_1 {widgetClassPath=&VerticalBox, parent=VerticalBox_0, widget=VerticalBox_1, slot {class=VerticalBoxSlot}}
  HB_SpotlightBassDrums {widgetClassPath=&HorizontalBox, parent=VerticalBox_1, widget=HB_SpotlightBassDrums, slot {class=VerticalBoxSlot}}
  SpotlightMarker_Bass {
   widgetClassPath=&Image
   bIsVariable=true
   parent=HB_SpotlightBassDrums
   widget=SpotlightMarker_Bass
   properties {
    toolTipText=Spotlight Competition
    visibility=SelfHitTestInvisible
    brush {
     bIsDynamicallyLoaded=false
     drawAs=Image
     imageType=NoImage
     mirroring=NoMirror
     resourceName=None
     resourceObject=/PilgrimQuickplay/Widgets/Materials/MI_UI_Icon_Instrument_Type_Bass_Sm
     tiling=NoTile
     imageSize=<32 32>
     margin=<ltrb 0 0 0 0>
     outlineSettings {
      bUseBrushTransparency=false
      roundingType=HalfHeightRadius
      width=0
      color {colorUseRule=UseColor_Specified, specifiedColor=<rgba 0 0 0 0>}
      cornerRadii=<0 0 0 0>
     }
     tintColor {colorUseRule=UseColor_Specified, specifiedColor=<rgba 0 0 0 1>}
     uVRegion {bIsValid=false, max=<0 0>, min=<0 0>}
    }
   }
   slot {class=HorizontalBoxSlot}
  }
  SpotlightMarker_Drums {
   widgetClassPath=&Image
   bIsVariable=true
   parent=HB_SpotlightBassDrums
   widget=SpotlightMarker_Drums
   properties {
    toolTipText=Spotlight Competition
    visibility=SelfHitTestInvisible
    brush {
     bIsDynamicallyLoaded=false
     drawAs=Image
     imageType=NoImage
     mirroring=NoMirror
     resourceName=None
     resourceObject=/PilgrimQuickplay/Widgets/Materials/MI_UI_Icon_Instrument_Type_Drums
     tiling=NoTile
     imageSize=<32 32>
     margin=<ltrb 0 0 0 0>
     outlineSettings {
      bUseBrushTransparency=false
      roundingType=HalfHeightRadius
      width=0
      color {colorUseRule=UseColor_Specified, specifiedColor=<rgba 0 0 0 0>}
      cornerRadii=<0 0 0 0>
     }
     tintColor {colorUseRule=UseColor_Specified, specifiedColor=<rgba 0 0 0 1>}
     uVRegion {bIsValid=false, max=<0 0>, min=<0 0>}
    }
   }
   slot {class=HorizontalBoxSlot}
  }
  HB_SpotlightGuitarVocals {widgetClassPath=&HorizontalBox, parent=VerticalBox_1, widget=HB_SpotlightGuitarVocals, slot {class=VerticalBoxSlot}}
  SpotlightMarker_Vocals {
   widgetClassPath=&Image
   bIsVariable=true
   parent=HB_SpotlightGuitarVocals
   widget=SpotlightMarker_Vocals
   properties {
    toolTipText=Spotlight Competition
    visibility=SelfHitTestInvisible
    brush {
     bIsDynamicallyLoaded=false
     drawAs=Image
     imageType=NoImage
     mirroring=NoMirror
     resourceName=None
     resourceObject=/PilgrimQuickplay/Widgets/Materials/MI_UI_Icon_Instrument_Type_Vocals_Sm
     tiling=NoTile
     imageSize=<32 32>
     margin=<ltrb 0 0 0 0>
     outlineSettings {
      bUseBrushTransparency=false
      roundingType=HalfHeightRadius
      width=0
      color {colorUseRule=UseColor_Specified, specifiedColor=<rgba 0 0 0 0>}
      cornerRadii=<0 0 0 0>
     }
     tintColor {colorUseRule=UseColor_Specified, specifiedColor=<rgba 0 0 0 1>}
     uVRegion {bIsValid=false, max=<0 0>, min=<0 0>}
    }
   }
   slot {class=HorizontalBoxSlot}
  }
  SpotlightMarker_Guitar {
   widgetClassPath=&Image
   bIsVariable=true
   parent=HB_SpotlightGuitarVocals
   widget=SpotlightMarker_Guitar
   properties {
    toolTipText=Spotlight Competition
    visibility=SelfHitTestInvisible
    brush {
     bIsDynamicallyLoaded=false
     drawAs=Image
     imageType=NoImage
     mirroring=NoMirror
     resourceName=None
     resourceObject=/PilgrimQuickplay/Widgets/Materials/MI_UI_Icon_Instrument_Type_Guitar_Sm
     tiling=NoTile
     imageSize=<32 32>
     margin=<ltrb 0 0 0 0>
     outlineSettings {
      bUseBrushTransparency=false
      roundingType=HalfHeightRadius
      width=0
      color {colorUseRule=UseColor_Specified, specifiedColor=<rgba 0 0 0 0>}
      cornerRadii=<0 0 0 0>
     }
     tintColor {colorUseRule=UseColor_Specified, specifiedColor=<rgba 0 0 0 1>}
     uVRegion {bIsValid=false, max=<0 0>, min=<0 0>}
    }
   }
   slot {class=HorizontalBoxSlot}
  }
  QuestMarker {
   widgetClassPath=&Image
   bIsVariable=true
   parent=VerticalBox_0
   widget=QuestMarker
   properties {
    visibility=Hidden
    brush {
     bIsDynamicallyLoaded=false
     drawAs=Image
     imageType=NoImage
     mirroring=NoMirror
     resourceName=None
     resourceObject=/FortniteCoreUI/Athena/HUD/Quests/Art/M_UI_TabCategory_XP_White
     tiling=NoTile
     imageSize=<32 32>
     margin=<ltrb 0 0 0 0>
     outlineSettings {
      bUseBrushTransparency=false
      roundingType=HalfHeightRadius
      width=0
      color {colorUseRule=UseColor_Specified, specifiedColor=<rgba 0 0 0 0>}
      cornerRadii=<0 0 0 0>
     }
     tintColor {colorUseRule=UseColor_Specified, specifiedColor=<rgba 1 1 1 1>}
     uVRegion {bIsValid=false, max=<0 0>, min=<0 0>}
    }
   }
   slot {class=VerticalBoxSlot}
  }
  AddedToSetlistIndicator {
   widgetClassPath=&Overlay
   bIsVariable=true
   parent=HorizontalBox_34
   widget=AddedToSetlistIndicator
   properties {visibility=Hidden}
   slot {class=HorizontalBoxSlot, properties {horizontalAlignment=HAlign_Right}}
  }
  AddedTag {
   widgetClassPath=&Image
   bIsVariable=true
   parent=AddedToSetlistIndicator
   widget=AddedTag
   properties {
    brush {
     bIsDynamicallyLoaded=false
     drawAs=Image
     imageType=NoImage
     mirroring=NoMirror
     resourceName=None
     resourceObject=/Game/UI/GreyBox/white
     tiling=NoTile
     imageSize=<16 16>
     margin=<ltrb 0 0 0 0>
     outlineSettings {
      bUseBrushTransparency=false
      roundingType=HalfHeightRadius
      width=0
      color {colorUseRule=UseColor_Specified, specifiedColor=<rgba 0 0 0 0>}
      cornerRadii=<0 0 0 0>
     }
     tintColor {colorUseRule=UseColor_Specified, specifiedColor=<rgba 1 1 1 1>}
     uVRegion {bIsValid=false, max=<0 0>, min=<0 0>}
    }
    colorAndOpacity=<rgba 0.6795 0.6514 0.0052 1>
   }
   slot {class=OverlaySlot, properties {horizontalAlignment=HAlign_Fill, verticalAlignment=VAlign_Fill}}
  }
  AlreadyAddedOverlay {
   widgetClassPath=&Overlay
   bIsVariable=true
   parent=Overlay_0
   widget=AlreadyAddedOverlay
   properties {visibility=Collapsed}
   slot {class=OverlaySlot, properties {horizontalAlignment=HAlign_Fill, verticalAlignment=VAlign_Fill}}
  }
  YellowBG {
   widgetClassPath=&Border
   parent=AlreadyAddedOverlay
   widget=YellowBG
   properties {
    horizontalAlignment=HAlign_Left
    verticalAlignment=VAlign_Center
    background {
     bIsDynamicallyLoaded=false
     drawAs=Image
     imageType=NoImage
     mirroring=NoMirror
     resourceName=None
     resourceObject=/FortUILibrary/Foundation/Art_ToolKit/AngledBoxes/M_UI_SimpleFeatured_YellowButton_NoShine
     tiling=NoTile
     imageSize=<32 32>
     margin=<ltrb 0 0 0 0>
     outlineSettings {
      bUseBrushTransparency=false
      roundingType=HalfHeightRadius
      width=0
      color {colorUseRule=UseColor_Specified, specifiedColor=<rgba 0 0 0 0>}
      cornerRadii=<0 0 0 0>
     }
     tintColor {colorUseRule=UseColor_Specified, specifiedColor=<rgba 1 1 1 1>}
     uVRegion {bIsValid=false, max=<0 0>, min=<0 0>}
    }
    padding=<ltrb 110 0 15 0>
   }
   slot {class=OverlaySlot, properties {verticalAlignment=VAlign_Bottom, padding=<ltrb 15 15 15 15>}}
  }
  AddedByPlayerText {
   widgetClassPath=&CommonTextBlock
   bIsVariable=true
   parent=YellowBG
   widget=AddedByPlayerText
   properties {
    justification=Center
    scrollStyle=&TextScrollStyle_PilgrimQuickplay_Setlist_C
    style=/FortUILibrary/Styles/Typography/Foundation/TextStyle-BurbankSmall-M-Black.TextStyle-BurbankSmall-M-Black_C
    text="Added by {0}"
    textTransformPolicy=ToUpper
    colorAndOpacity {colorUseRule=UseColor_Specified, specifiedColor=<rgba 0 0 0 1>}
    font {
     bForceMonospaced=false
     bMaterialIsStencil=false
     fontMaterial=null
     fontObject=&BurbankSmall
     letterSpacing=0
     monospacedWidth=1
     size=24
     skewAmount=0
     typefaceFontName=Black
     outlineSettings {
      bApplyOutlineToDropShadows=false
      bMiteredCorners=false
      bSeparateFillAlpha=false
      outlineMaterial=null
      outlineSize=0
      outlineColor=<rgba 0 0 0 1>
     }
    }
    shadowOffset=<0 0>
   }
   slot {class=BorderSlot, properties {horizontalAlignment=HAlign_Left, verticalAlignment=VAlign_Center, padding=<ltrb 110 0 15 0>}}
  }
  AddedByAvatar {
   widgetClassPath=&Image
   bIsVariable=true
   parent=AlreadyAddedOverlay
   widget=AddedByAvatar
   properties {brush {
     bIsDynamicallyLoaded=false
     drawAs=Image
     imageType=FullColor
     mirroring=NoMirror
     resourceName=None
     resourceObject=/SparksUICommon/Materials/MI_UI_BandMember_Avatar
     tiling=NoTile
     imageSize=<96 96>
     margin=<ltrb 0 0 0 0>
     outlineSettings {
      bUseBrushTransparency=false
      roundingType=HalfHeightRadius
      width=0
      color {colorUseRule=UseColor_Specified, specifiedColor=<rgba 0 0 0 0>}
      cornerRadii=<0 0 0 0>
     }
     tintColor {colorUseRule=UseColor_Specified, specifiedColor=<rgba 1 1 1 1>}
     uVRegion {bIsValid=false, max=<0 0>, min=<0 0>}
    }}
   slot {class=OverlaySlot, properties {verticalAlignment=VAlign_Bottom}}
  }
  MessageOverlay {
   widgetClassPath=&Overlay
   bIsVariable=true
   parent=Overlay_0
   widget=MessageOverlay
   properties {renderOpacity=0}
   slot {class=OverlaySlot, properties {horizontalAlignment=HAlign_Fill, verticalAlignment=VAlign_Fill}}
  }
  Border_80 {
   widgetClassPath=&Border
   parent=MessageOverlay
   widget=Border_80
   properties {
    background {
     bIsDynamicallyLoaded=false
     drawAs=Box
     imageType=NoImage
     mirroring=NoMirror
     resourceName=None
     resourceObject=null
     tiling=NoTile
     imageSize=<32 32>
     margin=<ltrb 0 0 0 0>
     outlineSettings {
      bUseBrushTransparency=false
      roundingType=FixedRadius
      width=0
      color {colorUseRule=UseColor_Specified, specifiedColor=<rgba 1 1 1 1>}
      cornerRadii=<20 20 20 20>
     }
     tintColor {colorUseRule=UseColor_Specified, specifiedColor=<rgba 1 1 1 1>}
     uVRegion {bIsValid=false, max=<0 0>, min=<0 0>}
    }
    brushColor=<rgba 0.0003 0.0762 0.4125 1>
   }
   slot {class=OverlaySlot, properties {horizontalAlignment=HAlign_Fill, verticalAlignment=VAlign_Fill, padding=<ltrb 25 25 25 25>}}
  }
  MessageText {
   widgetClassPath=&CommonTextBlock
   bIsVariable=true
   parent=MessageOverlay
   widget=MessageText
   properties {
    autoWrapText=true
    justification=Center
    style=&TextStyle_BurbankSmall_M_White_C
    text=MESSAGE TEXT
    textTransformPolicy=ToUpper
    font {
     bForceMonospaced=false
     bMaterialIsStencil=false
     fontMaterial=null
     fontObject=&BurbankSmall
     letterSpacing=0
     monospacedWidth=1
     size=24
     skewAmount=0
     typefaceFontName=Black
     outlineSettings {
      bApplyOutlineToDropShadows=false
      bMiteredCorners=false
      bSeparateFillAlpha=false
      outlineMaterial=null
      outlineSize=0
      outlineColor=<rgba 0 0 0 1>
     }
    }
    shadowOffset=<0 0>
   }
   slot {class=OverlaySlot, properties {horizontalAlignment=HAlign_Fill, verticalAlignment=VAlign_Center}}
  }
  InputActionWidget {widgetClassPath=&CommonActionWidget, bInherited=true, widget=null}
 ]
}
