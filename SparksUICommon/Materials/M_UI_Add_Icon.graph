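# Translucent UI material named as an add icon, built procedurally from signed-distance boxes.
# Two rounded MF_UI_Box bars share size and CornerRadius; the second uses UVs built from the G texture coordinate on both axes.
# MF_UI_AntiAliasing turns each distance field into an inner mask using Threshold.
# The summed, saturated masks are multiplied by the alpha of the RGB1 color parameter.
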Material /SparksUICommon/Materials/M_UI_Add_Icon
graph {
 properties {blendMode=BLEND_Translucent, materialDomain=MD_UI, pixelDepthOffsetMode=PDOM_Legacy, usedShadingModels=Default Lit}
 nodes [
  n1 = MaterialFunctionCall(
   DistanceField=MaterialFunctionCall(Size=Constant2Vector() {g=0.075, r=0.35}, CornerRoundness=ScalarParameter() {defaultValue=0.02, parameterName=CornerRadius}) {materialFunction=&MF_UI_Box}.SDF
   Threshold=ScalarParameter() {defaultValue=0.05, parameterName=Threshold}
  ) {materialFunction=&MF_UI_AntiAliasing}
  n2 = MaterialFunctionCall(Float2=TextureCoordinate()) {materialFunction=&BreakOutFloat2Components}
  n3 = MaterialFunctionCall(
   DistanceField=MaterialFunctionCall(
    UVs=MaterialFunctionCall(X=@n2.G, Y=@n2.G) {materialFunction=&MakeFloat2}.Result
    Size=Constant2Vector() {g=0.075, r=0.35}
    CornerRoundness=ScalarParameter() {defaultValue=0.02, parameterName=CornerRadius}
   ) {materialFunction=&MF_UI_Box}.SDF
   Threshold=ScalarParameter() {defaultValue=0.05, parameterName=Threshold}
  ) {materialFunction=&MF_UI_AntiAliasing}
  Multiply(A=VectorParameter() {parameterName=RGB1, defaultValue=<rgba 1 1 1 1>}.A, B=Saturate(None=Add(A=@n1.Inner, B=@n3.Inner) {constB=0}))
 ]
}
